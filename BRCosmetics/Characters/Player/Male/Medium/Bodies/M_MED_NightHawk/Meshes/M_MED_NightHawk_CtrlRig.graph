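# Control Rig for the NightHawk medium male player body, a corrective that shifts dynamic skirt bones as an arm moves.
# A spherical pose reader on upperarm_l measures the arm's direction against an active region with falloff.
# Its output weights a Modify Transforms step that offsets and rotates bones including dyn_SkirtA00_r and dyn_SkirtB00_r.
# The left arm drives right-suffixed skirt bones, and the graph has no second pose reader for the right arm.
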ControlRigBlueprint /BRCosmetics/Characters/Player/Male/Medium/Bodies/M_MED_NightHawk/Meshes/M_MED_NightHawk_CtrlRig
graphs [
 RigVMFunctionLibrary {links [], nodes []}
 RigVMModel {nodes [
   RigUnit_BeginExecution = &RigUnit_BeginExecution()
   n1 = &RigUnit_SphericalPoseReader(
    ActiveRegionScaleFactors=(PositiveWidth=1,NegativeWidth=1,PositiveHeight=1,NegativeHeight=1)
    ActiveRegionSize=0.1
    Debug=(bDrawDebug=false,DebugScale=25,DebugSegments=20,DebugThickness=0.25)
    DriverAxis=(X=1,Y=0,Z=1)
    DriverItem=(Type=Bone,Name="upperarm_l")
    FalloffRegionScaleFactors=(PositiveWidth=1,NegativeWidth=1,PositiveHeight=1,NegativeHeight=1)
    FalloffSize=0.2
    OptionalParentItem=(Type=Bone)
    RotationOffset=(X=90,Y=0,Z=-40)
    ExecutePin=@RigUnit_BeginExecution.ExecutePin
   )
   > &RigUnit_ModifyTransforms(
    ItemToModify="((Item=(Type=Bone,Name=\"dyn_SkirtA00_r\"),Transform=(Rotation=(X=0,Y=0,Z=0,W=1),Translation=(X=0,Y=-6,Z=0),Scale3D=(X=1,Y=1,Z=1))),(Item=(Type=Bone,Name=\"dyn_SkirtB00_r\"),Transform=(Rotation=(X=0,Y=0,Z=0,W=1),Translation=(X=0,Y=-6,Z=0),Scale3D=(X=1,Y=1,Z=1) ...(truncated)"
    ItemToModify.0.Transform.Rotation=&RigVMFunction_MathQuaternionFromEuler(Euler=(X=-10,Y=-10,Z=30)).Result
    ItemToModify.1.Transform.Rotation=&RigVMFunction_MathQuaternionFromEuler(Euler=(X=0,Y=0,Z=10)).Result
    Weight=@n1.OutputParam
   )
  ]}
]
hierarchy [
 root {loc=<0 0 0>, rot=<0 0 0 1>}
 pelvis {parent=root, loc=<0 -0.7071 93.1191>, rot=<0.0127 0.707 -0.0127 -0.707>}
 spine_01 {parent=pelvis, loc=<6.7494 -0.6388 0>, rot=<0 0 0.034 -0.9994>}
 spine_02 {parent=spine_01, loc=<7.2518 -0.035 0>, rot=<0 0 0.0238 -0.9997>}
 spine_03 {parent=spine_02, loc=<6.0104 0 0>, rot=<0 0 0.0149 -0.9999>}
 spine_04 {parent=spine_03, loc=<8.1024 0 0>, rot=<0 0 -0.0132 -0.9999>}
 spine_05 {parent=spine_04, loc=<10.0146 0 0>, rot=<0 0 -0.0015 -1>}
 attach_cape {parent=spine_05, loc=<9.017 -11.714 0>, rot=<0.0616 -0.7044 -0.0616 -0.7044>}
 neck_01 {parent=spine_05, loc=<11.3138 0.1529 0>, rot=<0 0 -0.2003 -0.9797>}
 neck_02 {parent=neck_01, loc=<4.5627 0 0>, rot=<0 0 0.0504 -0.9987>}
 head {parent=neck_02, loc=<5.3914 0 0>, rot=<0 0 0.115 -0.9934>}
 attach_backpack {parent=spine_05, loc=<-0.123 -13.473 0>, rot=<-0.0093 -0.707 0.0093 -0.707>}
 clavicle_l {parent=spine_05, loc=<4.3575 -0.1466 -2.3223>, rot=<0.0465 -0.7289 -0.009 -0.6829>}
 upperarm_l {parent=clavicle_l, loc=<15.5111 0 0>, rot=<0.0226 -0.3234 -0.0088 -0.9459>}
 lowerarm_l {parent=upperarm_l, loc=<26.6514 0 0>, rot=<0.0391 -0.0099 -0.2453 -0.9686>}
 hand_l {parent=lowerarm_l, loc=<22.3491 0 0>, rot=<-0.7081 -0.0507 0.0975 -0.6975>}
 thumb_01_l {parent=hand_l, loc=<1.3118 -0.3191 -2.459>, rot=<0.6546 -0.4042 0.0664 -0.6354>}
 thumb_02_l {parent=thumb_01_l, loc=<5.9336 -0.0448 0.0401>, rot=<-0.0037 0.0213 0.298 -0.9543>}
 thumb_03_l {parent=thumb_02_l, loc=<2.2223 0.3682 -0.0907>, rot=<0.0118 -0.013 0.0066 -0.9998>}
 index_metacarpal_l {parent=hand_l, loc=<1.6535 0.0183 -2.96>, rot=<0.0158 -0.0124 -0.0173 -0.9996>}
 index_01_l {parent=index_metacarpal_l, loc=<8.3008 0.0357 -0.0524>, rot=<-0.0044 0.0485 0.0924 -0.9945>}
 index_02_l {parent=index_01_l, loc=<3.3587 -0.3004 0.0452>, rot=<-0.002 0.0027 0.3225 -0.9466>}
 index_03_l {parent=index_02_l, loc=<2.5578 -0.123 -0.0252>, rot=<-0.002 0.0067 0.2558 -0.9667>}
 middle_metacarpal_l {parent=hand_l, loc=<2.1856 -0.1725 -0.5984>, rot=<0.006 0.0168 -0.0329 -0.9993>}
 middle_01_l {parent=middle_metacarpal_l, loc=<7.3275 -0.2447 0.0129>, rot=<-0.0208 0.0042 0.0858 -0.9961>}
 middle_02_l {parent=middle_01_l, loc=<4.0123 0.1401 0.1175>, rot=<-0.0218 0.0258 0.3512 -0.9357>}
 middle_03_l {parent=middle_02_l, loc=<2.6077 0.0534 0.1373>, rot=<0.0082 0.0188 0.1708 -0.9851>}
 ring_metacarpal_l {parent=hand_l, loc=<2.2074 -0.3625 1.3638>, rot=<-0.0284 0.0722 -0.0458 -0.9959>}
 ring_01_l {parent=ring_metacarpal_l, loc=<7.0295 -0.6197 -0.1446>, rot=<-0.0489 -0.0252 0.1892 -0.9804>}
 ring_02_l {parent=ring_01_l, loc=<3.6174 0.4647 -0.1027>, rot=<0 0 0.2806 -0.9598>}
 ring_03_l {parent=ring_02_l, loc=<2.4152 0.114 0.0199>, rot=<0 0 0.1189 -0.9929>}
 pinky_metacarpal_l {parent=hand_l, loc=<1.7616 -0.9753 3.3814>, rot=<-0.0397 0.0814 -0.069 -0.9935>}
 pinky_01_l {parent=pinky_metacarpal_l, loc=<7.1709 -0.4831 -0.1094>, rot=<-0.0654 -0.0035 0.2242 -0.9723>}
 pinky_02_l {parent=pinky_01_l, loc=<2.8274 0.0893 -0.1043>, rot=<0.0037 0.0011 0.2854 -0.9584>}
 pinky_03_l {parent=pinky_02_l, loc=<2.2046 0.0307 -0.0078>, rot=<-0.0002 0.0098 0.1445 -0.9895>}
 hand_attach_l {parent=hand_l, loc=<8.5 -3 0>, rot=<0.7071 0.7071 0 0>}
 weapon_l {parent=hand_l, loc=<-0.1974 -3.4111 0.3811>, rot=<0 0 0 1>}
 lowerarm_twist_02_l {parent=lowerarm_l, loc=<9.6199 0 0>, rot=<0 0 0 -1>}
 lowerarm_twist_01_l {parent=lowerarm_l, loc=<17.8549 0 0>, rot=<0 0 0 -1>}
 lowerarm_l_front {parent=lowerarm_l, loc=<0 0.5246 0>, rot=<0 0 0 -1>}
 lowerarm_l_back {parent=lowerarm_l, loc=<0 -0.4766 0>, rot=<0 0 0 -1>}
 upperarm_twist_01_l {parent=upperarm_l, loc=<0.0041 0 0>, rot=<0 0 0 1>}
 upperarm_twist_02_l {parent=upperarm_l, loc=<13.7941 0 0>, rot=<0 0 0 1>}
 up_uppelalm_l {parent=upperarm_l, loc=<-0.0003 -0.0004 0.5003>, rot=<0 0 0 1>}
 flont_uppelalm_l {parent=upperarm_l, loc=<-0.0001 0.4996 0.0001>, rot=<0 0 0 1>}
 back_uppelalm_l {parent=upperarm_l, loc=<-0.0001 -0.5 0.0001>, rot=<0 0 0 1>}
 upperarmDn_l {parent=clavicle_l, loc=<11.5695 -1.175 -9.5473>, rot=<0.0008 0.0336 -0.0252 -0.9991>}
 dyn_poncho_clav_roll_l_MSTR {parent=clavicle_l, loc=<14.2132 0 0>, rot=<0 0 0 -1>}
 dyn_poncho_clav_roll_l {parent=dyn_poncho_clav_roll_l_MSTR, loc=<0.3 0 0>, rot=<0 0 0 1>}
 dyn_poncho_bk_l_MSTR {parent=dyn_poncho_clav_roll_l, loc=<1.4188 -8.045 0.2425>, rot=<-0.4441 -0.4289 0.5268 -0.5842>}
 dyn_poncho_bk_l_01 {parent=dyn_poncho_bk_l_MSTR, loc=<0.3 0 0>, rot=<0 0 0 1>}
 dyn_poncho_bk_l_02 {parent=dyn_poncho_bk_l_01, loc=<7.2679 -0.1205 0.777>, rot=<0 -0.0751 0 -0.9972>}
 dyn_poncho_bk_l_03 {parent=dyn_poncho_bk_l_02, loc=<21.3019 0.1205 0.5318>, rot=<0 -0.0194 0 -0.9998>}
 dyn_poncho_bk_l_04 {parent=dyn_poncho_bk_l_03, loc=<23.964 0 0>, rot=<0 0 0 1>}
 clavicle_r {parent=spine_05, loc=<4.3575 -0.1466 2.3223>, rot=<0.7289 0.0465 -0.6829 0.009>}
 upperarm_r {parent=clavicle_r, loc=<-15.5111 0 0>, rot=<0.0226 -0.3234 -0.0088 -0.9459>}
 lowerarm_r {parent=upperarm_r, loc=<-26.6514 0 0>, rot=<0.0391 -0.0099 -0.2453 -0.9686>}
 hand_r {parent=lowerarm_r, loc=<-22.3491 0 0>, rot=<-0.7081 -0.0507 0.0975 -0.6975>}
 weapon_r {parent=hand_r, loc=<0.1974 3.4111 -0.3811>, rot=<0 0 0 1>}
 thumb_01_r {parent=hand_r, loc=<-1.3118 0.3191 2.459>, rot=<0.6546 -0.4042 0.0664 -0.6354>}
 thumb_02_r {parent=thumb_01_r, loc=<-5.9336 0.0448 -0.0401>, rot=<-0.0037 0.0213 0.298 -0.9543>}
 thumb_03_r {parent=thumb_02_r, loc=<-2.2223 -0.3682 0.0907>, rot=<0.0118 -0.013 0.0066 -0.9998>}
 index_metacarpal_r {parent=hand_r, loc=<-1.6535 -0.0183 2.96>, rot=<0.0158 -0.0124 -0.0173 -0.9996>}
 index_01_r {parent=index_metacarpal_r, loc=<-8.3008 -0.0357 0.0524>, rot=<-0.0044 0.0485 0.0924 -0.9945>}
 index_02_r {parent=index_01_r, loc=<-3.3587 0.3004 -0.0452>, rot=<-0.002 0.0027 0.3225 -0.9466>}
 index_03_r {parent=index_02_r, loc=<-2.5578 0.123 0.0252>, rot=<-0.002 0.0067 0.2558 -0.9667>}
 middle_metacarpal_r {parent=hand_r, loc=<-2.1856 0.1725 0.5984>, rot=<0.006 0.0168 -0.0329 -0.9993>}
 middle_01_r {parent=middle_metacarpal_r, loc=<-7.3275 0.2447 -0.0129>, rot=<-0.0208 0.0042 0.0858 -0.9961>}
 middle_02_r {parent=middle_01_r, loc=<-4.0123 -0.1401 -0.1175>, rot=<-0.0218 0.0258 0.3512 -0.9357>}
 middle_03_r {parent=middle_02_r, loc=<-2.6077 -0.0534 -0.1373>, rot=<0.0082 0.0188 0.1708 -0.9851>}
 ring_metacarpal_r {parent=hand_r, loc=<-2.2074 0.3625 -1.3638>, rot=<-0.0284 0.0722 -0.0458 -0.9959>}
 ring_01_r {parent=ring_metacarpal_r, loc=<-7.0295 0.6197 0.1446>, rot=<-0.0489 -0.0252 0.1892 -0.9804>}
 ring_02_r {parent=ring_01_r, loc=<-3.6174 -0.4647 0.1027>, rot=<0 0 0.2806 -0.9598>}
 ring_03_r {parent=ring_02_r, loc=<-2.4152 -0.114 -0.0199>, rot=<0 0 0.1189 -0.9929>}
 pinky_metacarpal_r {parent=hand_r, loc=<-1.7616 0.9753 -3.3814>, rot=<-0.0397 0.0814 -0.069 -0.9935>}
 pinky_01_r {parent=pinky_metacarpal_r, loc=<-7.1709 0.4831 0.1094>, rot=<-0.0654 -0.0035 0.2242 -0.9723>}
 pinky_02_r {parent=pinky_01_r, loc=<-2.8274 -0.0893 0.1043>, rot=<0.0037 0.0011 0.2854 -0.9584>}
 pinky_03_r {parent=pinky_02_r, loc=<-2.2046 -0.0307 0.0078>, rot=<-0.0002 0.0098 0.1445 -0.9895>}
 hand_attach_r {parent=hand_r, loc=<-8.5001 2.9999 0>, rot=<0 0 -0.7071 -0.7071>}
 lowerarm_twist_02_r {parent=lowerarm_r, loc=<-9.6199 0 0>, rot=<0 0 0 -1>}
 lowerarm_twist_01_r {parent=lowerarm_r, loc=<-17.8549 0 0>, rot=<0 0 0 -1>}
 lowerarm_r_front {parent=lowerarm_r, loc=<0 -0.4766 -0.0001>, rot=<0 0 0 -1>}
 lowerarm_r_back {parent=lowerarm_r, loc=<0 0.5246 -0.0001>, rot=<0 0 0 -1>}
 upperarm_twist_01_r {parent=upperarm_r, loc=<-0.0041 0 0>, rot=<0 0 0 -1>}
 upperarm_twist_02_r {parent=upperarm_r, loc=<-13.7941 0 0>, rot=<0 0 0 -1>}
 up_upperarm_r {parent=upperarm_r, loc=<0.0001 0.0004 -0.5001>, rot=<0 0 0 1>}
 front_upperarm_r {parent=upperarm_r, loc=<0.0001 -0.4996 -0.0001>, rot=<0 0 0 1>}
 back_upperarm_r {parent=upperarm_r, loc=<0.0001 0.5 -0.0001>, rot=<0 0 0 1>}
 upperarmDn_r {parent=clavicle_r, loc=<-11.5694 1.175 9.5468>, rot=<0.0008 0.0336 -0.0252 -0.9991>}
 dyn_poncho_clav_roll_r_MSTR {parent=clavicle_r, loc=<-14.2133 0 0.0001>, rot=<0 -1 0 0>}
 dyn_poncho_clav_roll_r {parent=dyn_poncho_clav_roll_r_MSTR, loc=<0.3 0 0>, rot=<0 0 0 1>}
 dyn_poncho_bk_r_MSTR {parent=dyn_poncho_clav_roll_r, loc=<1.4188 8.0451 0.2427>, rot=<0.4441 -0.4289 -0.5268 -0.5842>}
 dyn_poncho_bk_r_01 {parent=dyn_poncho_bk_r_MSTR, loc=<0.3004 0 -0.0001>, rot=<0 0 0 1>}
 dyn_poncho_bk_r_02 {parent=dyn_poncho_bk_r_01, loc=<7.2677 0.1206 0.7771>, rot=<0 -0.0751 0 -0.9972>}
 dyn_poncho_bk_r_03 {parent=dyn_poncho_bk_r_02, loc=<21.3022 -0.1206 0.5318>, rot=<0 -0.0194 0 -0.9998>}
 dyn_poncho_bk_r_04 {parent=dyn_poncho_bk_r_03, loc=<23.9638 0 0>, rot=<0 0 0 1>}
 dyn_poncho_bk_MSTR {parent=spine_05, loc=<9.2718 -9.2754 0>, rot=<0.1428 -0.6925 0.6925 -0.1428>}
 dyn_poncho_bk_01 {parent=dyn_poncho_bk_MSTR, loc=<0.3 0 0>, rot=<0 0 0 1>}
 dyn_poncho_bk_02 {parent=dyn_poncho_bk_01, loc=<10.6112 -0.1205 -0.0005>, rot=<0 -0.114 0 -0.9935>}
 dyn_poncho_bk_03 {parent=dyn_poncho_bk_02, loc=<23.8113 0.1205 0.3667>, rot=<0 -0.0328 0 -0.9995>}
 dyn_poncho_bk_04 {parent=dyn_poncho_bk_03, loc=<21.9319 0 0>, rot=<0 0 0 1>}
 dyn_collarB_01_r {parent=spine_05, loc=<12.2358 -1.1971 5.7727>, rot=<-0.6497 -0.0348 -0.0156 -0.7592>}
 dyn_collar_01 {parent=spine_05, loc=<12.5793 -5.4867 0.0968>, rot=<0.0208 0.0067 -0.0395 -0.999>}
 dyn_collarB_01_l {parent=spine_05, loc=<12.0883 -1.1649 -5.4601>, rot=<-0.6497 -0.0348 -0.0156 -0.7592>}
 dyn_cravat_01 {parent=spine_05, loc=<6.8443 7.3951 -0.0722>, rot=<-0.0275 -0.7179 0.6956 -0.0035>}
 dyn_cravat_02 {parent=dyn_cravat_01, loc=<5.047 0 -1.7461>, rot=<0 0 0 1>}
 dyn_hat_bk_MSTR {parent=spine_05, loc=<11.3555 -11.1844 -0.1452>, rot=<0.1428 -0.6925 0.6925 -0.1428>}
 dyn_hat_bk_01 {parent=dyn_hat_bk_MSTR, loc=<0.3 0 0>, rot=<0 0 0 1>}
 dyn_hat_bk_02 {parent=dyn_hat_bk_01, loc=<6.6971 -0.1205 0.5292>, rot=<0 -0.114 0 -0.9935>}
 dyn_SkirtA00_l {parent=spine_04, loc=<2.4947 9.5293 -11.9825>, rot=<-0.0119 -0.8935 0.4486 -0.0187>}
 dyn_SkirtA01_l {parent=dyn_SkirtA00_l, loc=<0.6904 -0.049 -0.0365>, rot=<0 0 0 1>}
 dyn_SkirtA02_l {parent=dyn_SkirtA01_l, loc=<5.1511 -0.2594 0.1301>, rot=<0 0 0 1>}
 dyn_SkirtA00_r {parent=spine_04, loc=<2.4947 9.5293 11.9213>, rot=<-0.0119 -0.8935 -0.4486 0.0187>}
 dyn_SkirtA01_r {parent=dyn_SkirtA00_r, loc=<0.6931 0 0>, rot=<0 0 0 1>}
 dyn_SkirtA02_r {parent=dyn_SkirtA01_r, loc=<5.1511 -0.2594 -0.1301>, rot=<0 0 0 1>}
 dyn_SkirtB00_l {parent=spine_04, loc=<2.4272 -1.2668 -14.4774>, rot=<0.0126 -0.7005 0.7122 -0.0433>}
 dyn_SkirtB01_l {parent=dyn_SkirtB00_l, loc=<0.6849 -0.049 -0.0943>, rot=<0 0 0 1>}
 dyn_SkirtB02_l {parent=dyn_SkirtB01_l, loc=<5.1168 -0.2827 -0.6246>, rot=<0 0 0 1>}
 dyn_SkirtB00_r {parent=spine_04, loc=<2.4872 -1.287 14.4774>, rot=<0.0433 -0.7122 -0.7005 0.0126>}
 dyn_SkirtB01_r {parent=dyn_SkirtB00_r, loc=<0.6849 0.049 0.0943>, rot=<0 0 0 1>}
 dyn_SkirtB02_r {parent=dyn_SkirtB01_r, loc=<5.1168 0.2827 0.6246>, rot=<0 0 0 1>}
 dyn_SkirtC00_l {parent=spine_04, loc=<2.7724 -10.0863 -6.1398>, rot=<-0.0209 -0.0135 0.9989 -0.04>}
 dyn_SkirtC01_l {parent=dyn_SkirtC00_l, loc=<0.6849 -0.0963 0.0449>, rot=<0 0 0 1>}
 dyn_SkirtC02_l {parent=dyn_SkirtC01_l, loc=<5.1168 0.0555 0.2552>, rot=<0 0 0 1>}
 dyn_SkirtC00_r {parent=spine_04, loc=<2.7724 -10.0863 6.1398>, rot=<-0.0254 0.1031 0.9936 -0.0373>}
 dyn_SkirtC01_r {parent=dyn_SkirtC00_r, loc=<0.6849 -0.0833 0.066>, rot=<0 0 0 1>}
 dyn_SkirtC02_r {parent=dyn_SkirtC01_r, loc=<5.1168 0.1319 0.3956>, rot=<0 0 0 1>}
 thigh_l {parent=pelvis, loc=<-2.5693 -0.0109 -7.7501>, rot=<0.0012 0.0572 0.0155 -0.9982>}
 calf_l {parent=thigh_l, loc=<-39.0484 0 0>, rot=<0.0006 -0.0161 -0.0391 -0.9991>}
 foot_l {parent=calf_l, loc=<-42.561 0.1077 0.4292>, rot=<0.0418 -0.0455 0.0343 -0.9975>}
 ball_l {parent=foot_l, loc=<-6.1599 13.8264 0.6806>, rot=<0.0001 -0.0001 -0.7063 -0.7079>}
 calf_twist_02_l {parent=calf_l, loc=<-18.8064 -0.0967 0.0397>, rot=<0.0237 0.0095 -0.0035 -0.9997>}
 calf_twist_01_l {parent=calf_l, loc=<-35.813 -0.0967 0.0397>, rot=<0.0237 0.0095 -0.0035 -0.9997>}
 calf_l_front {parent=calf_l, loc=<0 1 0>, rot=<0 0 0 -1>}
 calf_l_back {parent=calf_l, loc=<0 -1 0>, rot=<0 0 0 -1>}
 dyn_bootA_01_l {parent=calf_l, loc=<-3.5487 -0.8276 -4.5883>, rot=<0.0307 -0.7186 0.6947 -0.008>}
 dyn_bootA_rope01_l {parent=dyn_bootA_01_l, loc=<-6.629 0.7288 5.3775>, rot=<0 0 0 1>}
 dyn_bootA_rope02_l {parent=dyn_bootA_rope01_l, loc=<4.1733 0.1321 0.7835>, rot=<0 0 0 1>}
 dyn_bootB_01_l {parent=calf_l, loc=<-6.7359 3.8944 1.075>, rot=<0.0307 -0.7186 0.6947 -0.008>}
 dyn_bootC_01_l {parent=calf_l, loc=<-3.9651 -0.4367 5.9207>, rot=<0.0307 -0.7186 0.6947 -0.008>}
 dyn_bootC_rope01_l {parent=dyn_bootC_01_l, loc=<-7.1378 -2.464 3.8177>, rot=<0 0 0 1>}
 dyn_bootC_rope02_l {parent=dyn_bootC_rope01_l, loc=<3.7521 0 0.4299>, rot=<0 0 0 1>}
 thigh_twist_01_l {parent=thigh_l, loc=<-21.6303 0 0>, rot=<0.0013 0 0 -1>}
 thigh_volume_bk_l {parent=thigh_l, loc=<0 -1.5 0>, rot=<0 0 0 -1>}
 thigh_volume_fn_l {parent=thigh_l, loc=<0 1.5 0>, rot=<0.9998 0 -0.0217 0>}
 thigh_volume_lt_l {parent=thigh_l, loc=<0 0 -1.5>, rot=<-0.7069 -0.0154 0.0154 -0.7069>}
 thigh_r {parent=pelvis, loc=<-2.5693 -0.0109 7.7501>, rot=<0.0572 -0.0012 0.9982 0.0155>}
 calf_r {parent=thigh_r, loc=<39.0484 0 0>, rot=<0.0006 -0.0161 -0.0391 -0.9991>}
 foot_r {parent=calf_r, loc=<42.561 -0.1077 -0.4292>, rot=<0.0418 -0.0455 0.0343 -0.9975>}
 ball_r {parent=foot_r, loc=<6.1599 -13.8264 -0.6806>, rot=<0.0001 -0.0001 -0.7063 -0.7079>}
 calf_twist_02_r {parent=calf_r, loc=<18.8064 0.0967 -0.0397>, rot=<0.0237 0.0095 -0.0035 -0.9997>}
 calf_twist_01_r {parent=calf_r, loc=<35.813 0.0967 -0.0397>, rot=<0.0237 0.0095 -0.0035 -0.9997>}
 calf_r_front {parent=calf_r, loc=<0 -1 0>, rot=<0 0 0 -1>}
 calf_r_back {parent=calf_r, loc=<0 1 0>, rot=<0 0 0 -1>}
 dyn_bootA_01_r {parent=calf_r, loc=<3.5487 0.8276 4.5883>, rot=<0.0307 -0.7186 0.6947 -0.008>}
 dyn_bootA_rope01_r {parent=dyn_bootA_01_r, loc=<6.6291 -0.7288 -5.3775>, rot=<0 0 0 1>}
 dyn_bootA_rope02_r {parent=dyn_bootA_rope01_r, loc=<-4.1733 -0.1321 -0.7835>, rot=<0 0 0 1>}
 dyn_bootB_01_r {parent=calf_r, loc=<6.7358 -3.8944 -1.075>, rot=<0.0307 -0.7186 0.6947 -0.008>}
 dyn_bootC_01_r {parent=calf_r, loc=<3.9651 0.4367 -5.9207>, rot=<0.0307 -0.7186 0.6947 -0.008>}
 dyn_bootC_rope01_r {parent=dyn_bootC_01_r, loc=<7.1379 2.464 -3.8177>, rot=<0 0 0 1>}
 dyn_bootC_rope02_r {parent=dyn_bootC_rope01_r, loc=<-3.7521 0 -0.4299>, rot=<0 0 0 1>}
 thigh_twist_01_r {parent=thigh_r, loc=<21.6303 0 0>, rot=<0.0013 0 0 -1>}
 thigh_volume_bk_r {parent=thigh_r, loc=<-0.0002 1.5 0>, rot=<1 0 0 0>}
 thigh_volume_fn_r {parent=thigh_r, loc=<-0.0002 -1.5 0>, rot=<0 0 0 -1>}
 thigh_volume_rt_r {parent=thigh_r, loc=<-0.0002 0 1.5>, rot=<0.7071 0 0 -0.7071>}
]
