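# Physics asset for the NightHawk medium male player body, with capsule bodies for the upper body, pelvis, skirt bones and back poncho.
# Three poncho chains (center, left, right) hang from master bodies constrained with free rotation to spine_04 and the clavicles.
# Cross constraints tie each side chain to the center level by level, with free rotation on the lower two levels.
# Skirt, forearm and spine_03 bodies have no constraints.
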PhysicsAsset /BRCosmetics/Characters/Player/Male/Medium/Bodies/M_MED_NightHawk/Meshes/M_MED_NightHawk_Physics
bodies {
 clavicle_l [{shapeType=Capsule, length=8.6765, radius=9.1993, center=<11.2919 -0.8158 0.3274>, rotation=<rot 24.6602 -90 -90.221>}]
 clavicle_r [{shapeType=Capsule, length=5.327, radius=9.6701, center=<-8.5244 0.7077 0.4804>, rotation=<rot 23.8323 -90 -91.105>}]
 dyn_SkirtC01_l [{shapeType=Capsule, length=7, radius=8.75, shapeName=dyn_SkirtC01_l_capsule, center=<2.1315 -8.8631 0>}]
 dyn_SkirtC01_r [{shapeType=Capsule, length=7, radius=8.75, shapeName=dyn_SkirtC01_r_capsule, center=<2.1315 -8.8631 0>}]
 dyn_SkirtC02_l [{shapeType=Capsule, length=3.75, radius=8.75, shapeName=dyn_SkirtC02_l_capsule, center=<2.1315 -8.8631 0>}]
 dyn_SkirtC02_r [{shapeType=Capsule, length=3.75, radius=8.75, shapeName=dyn_SkirtC02_r_capsule, center=<2.1315 -8.8631 0>}]
 dyn_poncho_bk_01 [{shapeType=Capsule, length=12.5, radius=10.75, shapeName=dyn_poncho_bk_01_capsule, center=<10.1504 0 8.2589>, rotation=<rot -69.9999 -180 90>}]
 dyn_poncho_bk_02 [{
   shapeType=Capsule
   length=14.5
   radius=10.75
   shapeName=dyn_poncho_bk_02_capsule
   center=<13.8612 0 6.7766>
   rotation=<rot -90 175.2364 94.7636>
  }]
 dyn_poncho_bk_03 [{shapeType=Capsule, length=18, radius=10, shapeName=dyn_poncho_bk_03_capsule, center=<9.4009 0 5.2052>, rotation=<rot -90 -165.9638 75.9638>}]
 dyn_poncho_bk_04 [{
   shapeType=Capsule
   length=22.25
   radius=9.25
   shapeName=dyn_poncho_bk_04_capsule
   center=<3.4791 0 0.1852>
   rotation=<rot -90 172.875 97.125>
  }]
 dyn_poncho_bk_MSTR [{shapeType=Capsule, length=0.25, radius=8.75, shapeName=dyn_poncho_bk_MSTR_capsule, center=<5.5836 0 0.4885>, rotation=<rot -85.0001 0 0>}]
 dyn_poncho_bk_l_01 [{shapeType=Capsule, length=0, radius=10.75, shapeName=dyn_poncho_bk_l_01_capsule, center=<11.051 0 5.5565>, rotation=<rot -90 -180 90>}]
 dyn_poncho_bk_l_02 [{
   shapeType=Capsule
   length=3.5
   radius=10.75
   shapeName=dyn_poncho_bk_l_02_capsule
   center=<15.7605 -2.8218 4.9862>
   rotation=<rot -90 172.875 97.125>
  }]
 dyn_poncho_bk_l_03 [{
   shapeType=Capsule
   length=10.75
   radius=10.5
   shapeName=dyn_poncho_bk_l_03_capsule
   center=<9.4009 -5.2061 2.1252>
   rotation=<rot -90 180 90>
  }]
 dyn_poncho_bk_l_04 [{
   shapeType=Capsule
   length=6.75
   radius=9.25
   shapeName=dyn_poncho_bk_l_04_capsule
   center=<3.4791 0 0.1852>
   rotation=<rot -90 165.9638 104.0362>
  }]
 dyn_poncho_bk_l_MSTR [{shapeType=Capsule, length=0.25, radius=8.75, shapeName=dyn_poncho_bk_l_MSTR_capsule, center=<5.5836 0 0.4885>, rotation=<rot -85 0 0>}]
 dyn_poncho_bk_r_01 [{shapeType=Capsule, length=0, radius=10.75, shapeName=dyn_poncho_bk_r_01_capsule, center=<11.051 0 7.4915>, rotation=<rot -90 172.875 97.125>}]
 dyn_poncho_bk_r_02 [{shapeType=Capsule, length=9.25, radius=9.75, shapeName=dyn_poncho_bk_r_02_capsule, center=<15.7605 1.9195 3.8363>, rotation=<rot -90 180 90>}]
 dyn_poncho_bk_r_03 [{shapeType=Capsule, length=15.75, radius=9.5, shapeName=dyn_poncho_bk_r_03_capsule, center=<9.4009 3.0634 2.1252>, rotation=<rot -90 180 90>}]
 dyn_poncho_bk_r_04 [{shapeType=Capsule, length=10.75, radius=9.25, shapeName=dyn_poncho_bk_r_04_capsule, center=<3.4791 0 0.1852>, rotation=<rot -90 180 90>}]
 dyn_poncho_bk_r_MSTR [{shapeType=Capsule, length=0.25, radius=8.75, shapeName=dyn_poncho_bk_r_MSTR_capsule, center=<5.5836 0 0.4885>, rotation=<rot -85.0001 0 0>}]
 lowerarm_l [{
   shapeType=Capsule
   length=23.5
   radius=6.5
   shapeName=lowerarm_l_capsule
   center=<8.7103 1.0943 -0.7621>
   rotation=<rot -85 179.9996 -179.9996>
  }]
 lowerarm_r [{shapeType=Capsule, length=24, radius=6, shapeName=lowerarm_l_capsule, center=<-9.5738 -2.0166 0.8376>, rotation=<rot -85 179.9996 -179.9996>}]
 pelvis [{shapeType=Capsule, length=13.2129, radius=10.6835, center=<-3.2886 0.2567 -0.1249>, rotation=<rot 2.6452 -90.0001 -179.4429>}]
 spine_02 [{shapeType=Capsule, length=11.3202, radius=10.2772, center=<-3.4456 0.0993 0.0906>, rotation=<rot -0.7163 -90 0.2911>}]
 spine_03 [{shapeType=Capsule, length=8.5, radius=10, shapeName=spine_03_capsule, center=<2.1926 0 0>}]
 spine_04 [{shapeType=Capsule, length=14.5529, radius=12.2567, center=<7.1935 -0.0104 -0.0322>, rotation=<rot -0.9491 -90 -0.3245>}]
 upperarm_l [{shapeType=Capsule, length=16.9122, radius=5.686, center=<9.9055 0.2761 -0.6136>, rotation=<rot 0.273 -89.5634 -88.9811>}]
 upperarm_r [{shapeType=Capsule, length=19.995, radius=5.9348, center=<-10.7673 -0.5316 0.3494>, rotation=<rot 0.3335 -89.8133 -88.7548>}]
}
constraints [
 {bone1Name=spine_02, bone2Name=pelvis, swing1LimitDegrees=45, swing2LimitDegrees=45, twistLimitDegrees=45}
 {bone1Name=spine_04, bone2Name=spine_02, swing1LimitDegrees=45, swing2LimitDegrees=45, twistLimitDegrees=45}
 {bone1Name=clavicle_l, bone2Name=spine_04, swing1LimitDegrees=45, swing2LimitDegrees=45, twistLimitDegrees=45}
 {bone1Name=upperarm_l, bone2Name=clavicle_l, swing1LimitDegrees=45, swing2LimitDegrees=45, twistLimitDegrees=45}
 {bone1Name=dyn_poncho_bk_l_03, bone2Name=dyn_poncho_bk_l_02, swing1LimitDegrees=45, swing2LimitDegrees=45, twistLimitDegrees=45}
 {bone1Name=clavicle_r, bone2Name=spine_04, swing1LimitDegrees=45, swing2LimitDegrees=45, twistLimitDegrees=45}
 {bone1Name=upperarm_r, bone2Name=clavicle_r, swing1LimitDegrees=45, swing2LimitDegrees=45, twistLimitDegrees=45}
 {bone1Name=dyn_poncho_bk_03, bone2Name=dyn_poncho_bk_02, swing1LimitDegrees=45, swing2LimitDegrees=45, twistLimitDegrees=45}
 {bone1Name=dyn_poncho_bk_l_04, bone2Name=dyn_poncho_bk_l_03, swing1LimitDegrees=45, swing2LimitDegrees=45, twistLimitDegrees=45}
 {bone1Name=dyn_poncho_bk_l_02, bone2Name=dyn_poncho_bk_l_01, swing1LimitDegrees=45, swing2LimitDegrees=45, twistLimitDegrees=45}
 {bone1Name=dyn_poncho_bk_l_01, bone2Name=dyn_poncho_bk_l_MSTR, swing1LimitDegrees=45, swing2LimitDegrees=45, twistLimitDegrees=45}
 {bone1Name=dyn_poncho_bk_r_04, bone2Name=dyn_poncho_bk_r_03, swing1LimitDegrees=45, swing2LimitDegrees=45, twistLimitDegrees=45}
 {bone1Name=dyn_poncho_bk_r_03, bone2Name=dyn_poncho_bk_r_02, swing1LimitDegrees=45, swing2LimitDegrees=45, twistLimitDegrees=45}
 {bone1Name=dyn_poncho_bk_r_02, bone2Name=dyn_poncho_bk_r_01, swing1LimitDegrees=45, swing2LimitDegrees=45, twistLimitDegrees=45}
 {bone1Name=dyn_poncho_bk_r_01, bone2Name=dyn_poncho_bk_r_MSTR, swing1LimitDegrees=45, swing2LimitDegrees=45, twistLimitDegrees=45}
 {bone1Name=dyn_poncho_bk_04, bone2Name=dyn_poncho_bk_03, swing1LimitDegrees=45, swing2LimitDegrees=45, twistLimitDegrees=45}
 {bone1Name=dyn_poncho_bk_02, bone2Name=dyn_poncho_bk_01, swing1LimitDegrees=45, swing2LimitDegrees=45, twistLimitDegrees=45}
 {bone1Name=dyn_poncho_bk_01, bone2Name=dyn_poncho_bk_MSTR, swing1LimitDegrees=45, swing2LimitDegrees=45, twistLimitDegrees=45}
 {
  bone1Name=dyn_poncho_bk_MSTR
  bone2Name=spine_04
  swing1LimitDegrees=45
  swing1Motion=Free
  swing2LimitDegrees=45
  swing2Motion=Free
  twistLimitDegrees=45
  twistMotion=Free
 }
 {
  bone1Name=dyn_poncho_bk_r_MSTR
  bone2Name=clavicle_r
  swing1LimitDegrees=45
  swing1Motion=Free
  swing2LimitDegrees=45
  swing2Motion=Free
  twistLimitDegrees=45
  twistMotion=Free
 }
 {
  bone1Name=dyn_poncho_bk_l_MSTR
  bone2Name=clavicle_l
  swing1LimitDegrees=45
  swing1Motion=Free
  swing2LimitDegrees=45
  swing2Motion=Free
  twistLimitDegrees=45
  twistMotion=Free
 }
 {bone1Name=dyn_poncho_bk_r_01, bone2Name=dyn_poncho_bk_01, swing1LimitDegrees=45, swing2LimitDegrees=45, twistLimitDegrees=45}
 {bone1Name=dyn_poncho_bk_l_01, bone2Name=dyn_poncho_bk_01, swing1LimitDegrees=45, swing2LimitDegrees=45, twistLimitDegrees=45}
 {bone1Name=dyn_poncho_bk_r_02, bone2Name=dyn_poncho_bk_02, swing1LimitDegrees=45, swing2LimitDegrees=45, twistLimitDegrees=45}
 {bone1Name=dyn_poncho_bk_l_02, bone2Name=dyn_poncho_bk_02, swing1LimitDegrees=45, swing2LimitDegrees=45, twistLimitDegrees=45}
 {
  bone1Name=dyn_poncho_bk_r_04
  bone2Name=dyn_poncho_bk_04
  swing1LimitDegrees=45
  swing1Motion=Free
  swing2LimitDegrees=45
  swing2Motion=Free
  twistLimitDegrees=45
  twistMotion=Free
 }
 {
  bone1Name=dyn_poncho_bk_l_04
  bone2Name=dyn_poncho_bk_04
  swing1LimitDegrees=45
  swing1Motion=Free
  swing2LimitDegrees=45
  swing2Motion=Free
  twistLimitDegrees=45
  twistMotion=Free
 }
 {
  bone1Name=dyn_poncho_bk_l_03
  bone2Name=dyn_poncho_bk_03
  swing1LimitDegrees=45
  swing1Motion=Free
  swing2LimitDegrees=45
  swing2Motion=Free
  twistLimitDegrees=45
  twistMotion=Free
 }
 {
  bone1Name=dyn_poncho_bk_r_03
  bone2Name=dyn_poncho_bk_03
  swing1LimitDegrees=45
  swing1Motion=Free
  swing2LimitDegrees=45
  swing2Motion=Free
  twistLimitDegrees=45
  twistMotion=Free
 }
]
