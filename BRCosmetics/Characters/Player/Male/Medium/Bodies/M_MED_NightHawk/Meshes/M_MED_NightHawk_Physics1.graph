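# Physics asset for the NightHawk medium male player body, built around a three-chain back poncho.
# Capsules cover the torso, clavicles, arms, poncho and skirt bones; the skirt and left forearm bodies have no constraints.
# Center, left and right poncho chains hang from spine_05 and the clavicles with free-rotating roots and tight 01-02 joints.
# Cross-constraints tie both side chains to the center at their lower segments, linking the cloth-like panels.
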PhysicsAsset /BRCosmetics/Characters/Player/Male/Medium/Bodies/M_MED_NightHawk/Meshes/M_MED_NightHawk_Physics1
bodies {
 clavicle_l [{shapeType=Capsule, length=8.6765, radius=9.1993, center=<11.2919 -0.8158 0.3274>, rotation=<rot 24.6602 -90 -90.221>}]
 clavicle_r [{shapeType=Capsule, length=5.327, radius=9.6701, center=<-8.5244 0.7077 0.4804>, rotation=<rot 23.8323 -90 -91.105>}]
 dyn_SkirtC01_l [{shapeType=Capsule, length=8.25, radius=7.5, shapeName=dyn_SkirtC01_l_capsule, center=<2.1315 -9.8089 0>}]
 dyn_SkirtC01_r [{shapeType=Capsule, length=8.25, radius=7.5, shapeName=dyn_SkirtC01_r_capsule, center=<2.1315 -9.8089 0>}]
 dyn_SkirtC02_l [{shapeType=Capsule, length=8.25, radius=7.5, shapeName=dyn_SkirtC02_l_capsule, center=<2.1315 -9.8089 0>}]
 dyn_SkirtC02_r [{shapeType=Capsule, length=8.25, radius=7.5, shapeName=dyn_SkirtC02_r_capsule, center=<2.1315 -9.8089 0>}]
 dyn_poncho_bk_01 [{shapeType=Capsule, length=4.5, radius=10.75, shapeName=dyn_poncho_bk_01_capsule, center=<11.051 0 5.5565>, rotation=<rot -90 180 90>}]
 dyn_poncho_bk_02 [{shapeType=Capsule, length=4.5, radius=10.75, shapeName=dyn_poncho_bk_02_capsule, center=<15.7605 0 5.0769>, rotation=<rot -90 -180 90>}]
 dyn_poncho_bk_03 [{
   shapeType=Capsule
   length=14.5
   radius=11
   shapeName=dyn_poncho_bk_03_capsule
   center=<9.4009 0 3.6806>
   rotation=<rot -90 -165.9638 75.9638>
  }]
 dyn_poncho_bk_04 [{
   shapeType=Capsule
   length=11.25
   radius=8.75
   shapeName=dyn_poncho_bk_04_capsule
   center=<3.4791 0 1.3553>
   rotation=<rot -90 172.875 97.125>
  }]
 dyn_poncho_bk_MSTR [{shapeType=Capsule, length=0.25, radius=8.75, shapeName=dyn_poncho_bk_MSTR_capsule, center=<5.5836 0 0.4885>, rotation=<rot -85.0001 0 0>}]
 dyn_poncho_bk_l_01 [{shapeType=Capsule, length=0, radius=10.75, shapeName=dyn_poncho_bk_l_01_capsule, center=<11.051 0 5.5565>, rotation=<rot -90 -180 90>}]
 dyn_poncho_bk_l_02 [{
   shapeType=Capsule
   length=5.75
   radius=10.75
   shapeName=dyn_poncho_bk_l_02_capsule
   center=<15.7605 0 3.8363>
   rotation=<rot -90 172.875 97.125>
  }]
 dyn_poncho_bk_l_03 [{shapeType=Capsule, length=7, radius=11, shapeName=dyn_poncho_bk_l_03_capsule, center=<9.4009 0 2.44>, rotation=<rot -90 180 90>}]
 dyn_poncho_bk_l_04 [{
   shapeType=Capsule
   length=11.25
   radius=8.75
   shapeName=dyn_poncho_bk_l_04_capsule
   center=<3.4791 0 1.3553>
   rotation=<rot -90 165.9638 104.0362>
  }]
 dyn_poncho_bk_l_MSTR [{shapeType=Capsule, length=0.25, radius=8.75, shapeName=dyn_poncho_bk_l_MSTR_capsule, center=<5.5836 0 0.4885>, rotation=<rot -85 0 0>}]
 dyn_poncho_bk_r_01 [{shapeType=Capsule, length=0, radius=10.75, shapeName=dyn_poncho_bk_r_01_capsule, center=<11.051 0 5.5565>, rotation=<rot -90 180 90>}]
 dyn_poncho_bk_r_02 [{
   shapeType=Capsule
   length=6
   radius=9.5
   shapeName=dyn_poncho_bk_r_02_capsule
   center=<15.7605 0 3.8363>
   rotation=<rot -90 175.2364 94.7636>
  }]
 dyn_poncho_bk_r_03 [{shapeType=Capsule, length=4, radius=11, shapeName=dyn_poncho_bk_r_03_capsule, center=<9.4009 0 2.44>, rotation=<rot -90 165.9638 104.0362>}]
 dyn_poncho_bk_r_04 [{shapeType=Capsule, length=11.25, radius=8.75, shapeName=dyn_poncho_bk_r_04_capsule, center=<3.4791 0 1.3553>, rotation=<rot -90 180 90>}]
 dyn_poncho_bk_r_MSTR [{shapeType=Capsule, length=0.25, radius=8.75, shapeName=dyn_poncho_bk_r_MSTR_capsule, center=<5.5836 0 0.4885>, rotation=<rot -85.0001 0 0>}]
 lowerarm_l [{shapeType=Capsule, length=25.5, radius=4.5, shapeName=lowerarm_l_capsule, center=<12.8709 0 -1.1261>, rotation=<rot -85 179.9996 -179.9996>}]
 pelvis [{shapeType=Capsule, length=12.7129, radius=11.1835, center=<-2.5368 -0.3638 -0.1035>, rotation=<rot 2.6452 -90.0001 -179.4429>}]
 spine_02 [{shapeType=Capsule, length=9.5702, radius=12.0272, center=<2.0246 0.2809 0.0651>, rotation=<rot -0.7163 -90 0.2911>}]
 spine_04 [{shapeType=Capsule, length=15.5529, radius=11.2567, center=<0.8901 0.0799 -0.0664>, rotation=<rot -0.9491 -90 -0.3245>}]
 spine_05 [{shapeType=Capsule, length=10.0359, radius=12.8829, center=<-2.7652 2.8207 -0.0673>, rotation=<rot 2.0249 -84.0759 -90.5428>}]
 upperarm_l [{shapeType=Capsule, length=17.1622, radius=5.436, center=<9.9055 0.2761 -0.6136>, rotation=<rot 0.273 -89.5634 -88.9811>}]
 upperarm_r [{shapeType=Capsule, length=20.495, radius=2.9348, center=<-14.9653 -0.5447 0.4406>, rotation=<rot 0.3335 -89.8133 -88.7548>}]
}
constraints [
 {bone1Name=spine_02, bone2Name=pelvis, swing1LimitDegrees=45, swing2LimitDegrees=45, twistLimitDegrees=45}
 {bone1Name=spine_04, bone2Name=spine_02, swing1LimitDegrees=45, swing2LimitDegrees=45, twistLimitDegrees=45}
 {bone1Name=spine_05, bone2Name=spine_04, swing1LimitDegrees=45, swing2LimitDegrees=45, twistLimitDegrees=45}
 {bone1Name=clavicle_l, bone2Name=spine_05, swing1LimitDegrees=45, swing2LimitDegrees=45, twistLimitDegrees=45}
 {bone1Name=upperarm_l, bone2Name=clavicle_l, swing1LimitDegrees=45, swing2LimitDegrees=45, twistLimitDegrees=45}
 {bone1Name=dyn_poncho_bk_l_03, bone2Name=dyn_poncho_bk_l_02, swing1LimitDegrees=45, swing2LimitDegrees=45, twistLimitDegrees=45}
 {bone1Name=clavicle_r, bone2Name=spine_05, swing1LimitDegrees=45, swing2LimitDegrees=45, twistLimitDegrees=45}
 {bone1Name=upperarm_r, bone2Name=clavicle_r, swing1LimitDegrees=45, swing2LimitDegrees=45, twistLimitDegrees=45}
 {bone1Name=dyn_poncho_bk_03, bone2Name=dyn_poncho_bk_02, swing1LimitDegrees=45, swing2LimitDegrees=45, twistLimitDegrees=45}
 {bone1Name=dyn_poncho_bk_l_04, bone2Name=dyn_poncho_bk_l_03, swing1LimitDegrees=45, swing2LimitDegrees=45, twistLimitDegrees=45}
 {bone1Name=dyn_poncho_bk_l_02, bone2Name=dyn_poncho_bk_l_01, swing1LimitDegrees=5, swing2LimitDegrees=5, twistLimitDegrees=5}
 {bone1Name=dyn_poncho_bk_l_01, bone2Name=dyn_poncho_bk_l_MSTR, swing1LimitDegrees=45, swing2LimitDegrees=45, twistLimitDegrees=45}
 {bone1Name=dyn_poncho_bk_r_04, bone2Name=dyn_poncho_bk_r_03, swing1LimitDegrees=45, swing2LimitDegrees=45, twistLimitDegrees=45}
 {bone1Name=dyn_poncho_bk_r_03, bone2Name=dyn_poncho_bk_r_02, swing1LimitDegrees=45, swing2LimitDegrees=45, twistLimitDegrees=45}
 {bone1Name=dyn_poncho_bk_r_02, bone2Name=dyn_poncho_bk_r_01, swing1LimitDegrees=5, swing2LimitDegrees=5, twistLimitDegrees=5}
 {bone1Name=dyn_poncho_bk_r_01, bone2Name=dyn_poncho_bk_r_MSTR, swing1LimitDegrees=45, swing2LimitDegrees=45, twistLimitDegrees=45}
 {bone1Name=dyn_poncho_bk_04, bone2Name=dyn_poncho_bk_03, swing1LimitDegrees=45, swing2LimitDegrees=45, twistLimitDegrees=45}
 {bone1Name=dyn_poncho_bk_02, bone2Name=dyn_poncho_bk_01, swing1LimitDegrees=5, swing2LimitDegrees=5, twistLimitDegrees=5}
 {bone1Name=dyn_poncho_bk_01, bone2Name=dyn_poncho_bk_MSTR, swing1LimitDegrees=45, swing2LimitDegrees=45, twistLimitDegrees=45}
 {
  bone1Name=dyn_poncho_bk_MSTR
  bone2Name=spine_05
  swing1LimitDegrees=45
  swing1Motion=Free
  swing2LimitDegrees=45
  swing2Motion=Free
  twistLimitDegrees=45
  twistMotion=Free
 }
 {
  bone1Name=dyn_poncho_bk_r_MSTR
  bone2Name=clavicle_r
  swing1LimitDegrees=45
  swing1Motion=Free
  swing2LimitDegrees=45
  swing2Motion=Free
  twistLimitDegrees=45
  twistMotion=Free
 }
 {
  bone1Name=dyn_poncho_bk_l_MSTR
  bone2Name=clavicle_l
  swing1LimitDegrees=45
  swing1Motion=Free
  swing2LimitDegrees=45
  swing2Motion=Free
  twistLimitDegrees=45
  twistMotion=Free
 }
 {bone1Name=dyn_poncho_bk_r_03, bone2Name=dyn_poncho_bk_03, swing1LimitDegrees=45, swing2LimitDegrees=45, twistLimitDegrees=45}
 {bone1Name=dyn_poncho_bk_l_03, bone2Name=dyn_poncho_bk_03, swing1LimitDegrees=45, swing2LimitDegrees=45, twistLimitDegrees=45}
 {bone1Name=dyn_poncho_bk_r_04, bone2Name=dyn_poncho_bk_04, swing1LimitDegrees=45, swing2LimitDegrees=45, twistLimitDegrees=45}
 {bone1Name=dyn_poncho_bk_l_04, bone2Name=dyn_poncho_bk_04, swing1LimitDegrees=45, swing2LimitDegrees=45, twistLimitDegrees=45}
]
