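# Animation Blueprint for the NightHawk male player body cosmetic, adding secondary motion and corrective bones to a pose copied from the parent mesh.
# Constraints and variable-driven bone transforms adjust attachments, collar and thigh volumes, while AnimDynamics swings cravat, boot and skirt bones.
# Cached poses feed RigidBody simulations with emote-dependent blending.
# The output switches on cape-hide and backpack flags, shrinking poncho bones to hide them or selecting a physics pass.
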AnimBlueprint /BRCosmetics/Characters/Player/Male/Medium/Bodies/M_MED_NightHawk/Meshes/M_MED_NightHawk_AnimBP
target_skeleton=&/M_MED_NightHawk_Skeleton
graphs [
 AnimGraph {nodes [
   n1 = "Transform (Modify) Bone - Bone: thigh_Tshirt_bk_l"(Alpha=Get thigh_shape_bk_l() -> (thigh_shape_bk_l:double), Translation="0,0,0", Rotation="0,0,0", Scale="1,1,1") {
    rotationMode=BMM_Additive
    rotationSpace=BCS_BoneSpace
    scaleMode=BMM_Additive
    scaleSpace=BCS_BoneSpace
    translationMode=BMM_Additive
    translationSpace=BCS_BoneSpace
    alphaBoolBlend {blendOption=Linear}
    boneToModify {boneName=thigh_Tshirt_bk_l}
   }
   "Transform (Modify) Bone - Bone: thigh_Tshirt_bk_r"(ComponentPose=@n1, Alpha=Get thigh_shape_bk_r() -> (thigh_shape_bk_r:double), Translation="0,0,0", Rotation="0,0,0", Scale="1,1,1") {
    rotationMode=BMM_Additive
    rotationSpace=BCS_BoneSpace
    scaleMode=BMM_Additive
    scaleSpace=BCS_BoneSpace
    translationMode=BMM_Additive
    translationSpace=BCS_BoneSpace
    alphaBoolBlend {blendOption=Linear}
    boneToModify {boneName=thigh_Tshirt_bk_r}
   }
   n2 = RigidBody!(ComponentPose=Local To Component(LocalPose=Use cached pose 'PostEmoteCache'()), Alpha, OverrideWorldGravity:Vector="0,0,-980", ExternalForce) -> (Pose) {
    bClampLinearTranslationLimitToRefPose=true
    bOverrideWorldGravity=true
    overridePhysicsAsset=&/M_MED_NightHawk_Physics1
    simulationSpace=BaseBoneSpace
    alphaBoolBlend {blendOption=Linear}
    alphaScaleBiasClamp {bInterpResult=true}
    baseBoneRef {boneName=spine_02}
    componentAppliedLinearAccClamp=<400 400 400>
    componentLinearAccScale=<0.01 0.01 0.01>
    componentLinearVelScale=<0.01 0.01 0.01>
    overrideWorldGravity {z=-980}
   }
   n3 = Component To Local(ComponentPose=@n2)
   n4 = RigidBody(ComponentPose=Local To Component(LocalPose=Use cached pose 'PostEmoteCache'()), Alpha=0, OverrideWorldGravity="0,0,-400") {
    bClampLinearTranslationLimitToRefPose=true
    bOverrideWorldGravity=true
    overridePhysicsAsset=&/M_MED_NightHawk_Physics
    simulationSpace=BaseBoneSpace
    alphaBoolBlend {blendOption=Linear}
    alphaScaleBiasClamp {bInterpResult=true}
    baseBoneRef {boneName=pelvis}
    componentAppliedLinearAccClamp=<2000 2000 2000>
    componentLinearAccScale=<0.35 0.3 0.1>
    componentLinearVelScale=<0.35 0.3 0.1>
    overrideWorldGravity {z=-980}
   }
   n5 = Component To Local(ComponentPose=@n4)
   n6 = Get isWearingPackBool() -> (isWearingPackBool:bool)
   n7 = "Transform (Modify) Bone - Bone: dyn_poncho_clav_roll_l_MSTR"(
    ComponentPose=Local To Component(LocalPose=Use cached pose 'MainSetup'())
    Alpha=1
    Translation="0,0,0"
    Rotation="0,0,0"
    Scale="0.01,0.01,0.01"
   ) {
    rotationMode=BMM_Additive
    rotationSpace=BCS_BoneSpace
    scaleMode=BMM_Additive
    scaleSpace=BCS_BoneSpace
    translationMode=BMM_Additive
    translationSpace=BCS_BoneSpace
    alphaBoolBlend {blendOption=Linear}
    alphaScaleBiasClamp {bInterpResult=true}
    boneToModify {boneName=dyn_poncho_clav_roll_l_MSTR}
    scale=<0.01 0.01 0.01>
   }
   n8 = "Transform (Modify) Bone - Bone: dyn_poncho_clav_roll_r_MSTR"(ComponentPose=@n7, Alpha=1, Translation="0,0,0", Rotation="0,0,0", Scale="0.01,0.01,0.01") {
    rotationMode=BMM_Additive
    rotationSpace=BCS_BoneSpace
    scaleMode=BMM_Additive
    scaleSpace=BCS_BoneSpace
    translationMode=BMM_Additive
    translationSpace=BCS_BoneSpace
    alphaBoolBlend {blendOption=Linear}
    alphaScaleBiasClamp {bInterpResult=true}
    boneToModify {boneName=dyn_poncho_clav_roll_r_MSTR}
    scale=<0.01 0.01 0.01>
   }
   n9 = "Transform (Modify) Bone - Bone: dyn_poncho_bk_MSTR"(ComponentPose=@n8, Alpha=1, Translation="0,0,0", Rotation="0,0,0", Scale="0.01,0.01,0.01") {
    rotationMode=BMM_Additive
    rotationSpace=BCS_BoneSpace
    scaleMode=BMM_Additive
    scaleSpace=BCS_BoneSpace
    translationMode=BMM_Additive
    translationSpace=BCS_BoneSpace
    alphaBoolBlend {blendOption=Linear}
    alphaScaleBiasClamp {bInterpResult=true}
    boneToModify {boneName=dyn_poncho_bk_MSTR}
    scale=<0.01 0.01 0.01>
   }
   Output Pose(Result=Blend Poses by bool(
     BlendPose_0=Component To Local(ComponentPose=@n9)
     BlendPose_1=Blend Poses by bool(BlendPose_0=@n3, BlendPose_1=@n5, bActiveValue=@n6) {blendType=Linear}
     bActiveValue=Get CapeHide_Tag() -> (CapeHide_Tag:bool)
    ) {blendType=Linear}) {layerGroup=None}
   n10 = Anim Dynamics - Bone: dyn_SkirtA01_l(ComponentPose=Local To Component(LocalPose=Use cached pose 'CrouchingCache'()), Alpha=0.5, ExternalForce="0,0,0") {
    angularDampingOverride=0.999
    angularSpringConstant=800
    bAngularSpring=true
    bGravityOverrideInSimSpace=true
    bOverrideAngularDamping=true
    bUseGravityOverride=true
    gravityScale=0.2
    simulationSpace=RootRelative
    alphaBoolBlend {blendOption=Linear}
    boundBone {boneName=dyn_SkirtA01_l}
    componentAppliedLinearAccClamp=<4000 4000 4000>
    componentLinearAccScale=<0.8 0.8 0.8>
    componentLinearVelScale=<0.8 0.8 0.8>
    gravityOverride {z=-980}
    physicsBodyDefinitions [{
      sphereCollisionRadius=0
      boundBone {boneName=dyn_SkirtA01_l}
      boxExtents=<25 25 25>
      constraintSetup {angularTargetAxis=AxisY, angularLimitsMax=<5 1 60>, angularLimitsMin=<-5 -1 -25>, angularTarget=<0 0 0>}
      localJointOffset=<5 0.5 -0.2>
     }]
    relativeSpaceBone {boneName=spine_05}
   }
   n11 = Anim Dynamics - Bone: dyn_SkirtA02_l(ComponentPose=@n10, Alpha=0.5, ExternalForce="0,0,0") {
    angularDampingOverride=0.999
    angularSpringConstant=800
    bAngularSpring=true
    bGravityOverrideInSimSpace=true
    bOverrideAngularDamping=true
    bUseGravityOverride=true
    gravityScale=0.2
    simulationSpace=RootRelative
    alphaBoolBlend {blendOption=Linear}
    boundBone {boneName=dyn_SkirtA02_l}
    componentAppliedLinearAccClamp=<4000 4000 4000>
    componentLinearAccScale=<0.8 0.8 0.8>
    componentLinearVelScale=<0.8 0.8 0.8>
    gravityOverride {z=-980}
    physicsBodyDefinitions [{
      sphereCollisionRadius=0
      boundBone {boneName=dyn_SkirtA02_l}
      boxExtents=<25 25 25>
      constraintSetup {angularTargetAxis=AxisY, angularLimitsMax=<5 1 60>, angularLimitsMin=<-5 -1 -25>, angularTarget=<0 0 0>}
      localJointOffset=<5 0.5 -0.2>
     }]
    relativeSpaceBone {boneName=spine_05}
   }
   n12 = Anim Dynamics - Bone: dyn_SkirtA01_r(ComponentPose=@n11, Alpha=0.5, ExternalForce="0,0,0") {
    angularDampingOverride=0.999
    angularSpringConstant=800
    bAngularSpring=true
    bGravityOverrideInSimSpace=true
    bOverrideAngularDamping=true
    bUseGravityOverride=true
    gravityScale=0.2
    simulationSpace=RootRelative
    alphaBoolBlend {blendOption=Linear}
    boundBone {boneName=dyn_SkirtA01_r}
    componentAppliedLinearAccClamp=<4000 4000 4000>
    componentLinearAccScale=<0.8 0.8 0.8>
    componentLinearVelScale=<0.8 0.8 0.8>
    gravityOverride {z=-980}
    physicsBodyDefinitions [{
      sphereCollisionRadius=0
      boundBone {boneName=dyn_SkirtA01_r}
      boxExtents=<25 25 25>
      constraintSetup {angularTargetAxis=AxisY, angularLimitsMax=<5 1 60>, angularLimitsMin=<-5 -1 -25>, angularTarget=<0 0 0>}
      localJointOffset=<5 0.5 0.2>
     }]
    relativeSpaceBone {boneName=spine_05}
   }
   n13 = Anim Dynamics - Bone: dyn_SkirtA02_r(ComponentPose=@n12, Alpha=0.5, ExternalForce="0,0,0") {
    angularDampingOverride=0.999
    angularSpringConstant=800
    bAngularSpring=true
    bGravityOverrideInSimSpace=true
    bOverrideAngularDamping=true
    bUseGravityOverride=true
    gravityScale=0.2
    simulationSpace=RootRelative
    alphaBoolBlend {blendOption=Linear}
    boundBone {boneName=dyn_SkirtA02_r}
    componentAppliedLinearAccClamp=<4000 4000 4000>
    componentLinearAccScale=<0.8 0.8 0.8>
    componentLinearVelScale=<0.8 0.8 0.8>
    gravityOverride {z=-980}
    physicsBodyDefinitions [{
      sphereCollisionRadius=0
      boundBone {boneName=dyn_SkirtA02_r}
      boxExtents=<25 25 25>
      constraintSetup {angularTargetAxis=AxisY, angularLimitsMax=<5 1 60>, angularLimitsMin=<-5 -1 -25>, angularTarget=<0 0 0>}
      localJointOffset=<5 0.5 0.2>
     }]
    relativeSpaceBone {boneName=spine_05}
   }
   n14 = Anim Dynamics - Bone: dyn_SkirtB01_l(ComponentPose=@n13, Alpha=0.5, ExternalForce="0,0,0") {
    angularDampingOverride=0.999
    angularSpringConstant=800
    bAngularSpring=true
    bGravityOverrideInSimSpace=true
    bOverrideAngularDamping=true
    bUseGravityOverride=true
    gravityScale=0.2
    simulationSpace=RootRelative
    alphaBoolBlend {blendOption=Linear}
    boundBone {boneName=dyn_SkirtB01_l}
    componentAppliedLinearAccClamp=<3000 3000 3000>
    componentLinearAccScale=<0.8 0.8 0.8>
    componentLinearVelScale=<0.8 0.8 0.8>
    gravityOverride {z=-980}
    physicsBodyDefinitions [{
      sphereCollisionRadius=0
      boundBone {boneName=dyn_SkirtB01_l}
      boxExtents=<25 25 25>
      constraintSetup {angularTargetAxis=AxisY, angularLimitsMax=<5 1 60>, angularLimitsMin=<-5 -1 -15>, angularTarget=<0 0 0>}
      localJointOffset=<5 0.5 -0.2>
     }]
    relativeSpaceBone {boneName=spine_05}
   }
   n15 = Anim Dynamics - Bone: dyn_SkirtB02_l(ComponentPose=@n14, Alpha=0.5, ExternalForce="0,0,0") {
    angularDampingOverride=0.999
    angularSpringConstant=800
    bAngularSpring=true
    bGravityOverrideInSimSpace=true
    bOverrideAngularDamping=true
    bUseGravityOverride=true
    gravityScale=0.2
    simulationSpace=RootRelative
    alphaBoolBlend {blendOption=Linear}
    boundBone {boneName=dyn_SkirtB02_l}
    componentAppliedLinearAccClamp=<3000 3000 3000>
    componentLinearAccScale=<0.8 0.8 0.8>
    componentLinearVelScale=<0.8 0.8 0.8>
    gravityOverride {z=-980}
    physicsBodyDefinitions [{
      sphereCollisionRadius=0
      boundBone {boneName=dyn_SkirtB02_l}
      boxExtents=<25 25 25>
      constraintSetup {angularTargetAxis=AxisY, angularLimitsMax=<5 1 60>, angularLimitsMin=<-5 -1 -15>, angularTarget=<0 0 0>}
      localJointOffset=<5 0.5 -0.2>
     }]
    relativeSpaceBone {boneName=spine_05}
   }
   n16 = Anim Dynamics - Bone: dyn_SkirtB01_r(ComponentPose=@n15, Alpha=0.5, ExternalForce="0,0,0") {
    angularDampingOverride=0.999
    angularSpringConstant=800
    bAngularSpring=true
    bGravityOverrideInSimSpace=true
    bOverrideAngularDamping=true
    bUseGravityOverride=true
    gravityScale=0.2
    simulationSpace=RootRelative
    alphaBoolBlend {blendOption=Linear}
    boundBone {boneName=dyn_SkirtB01_r}
    componentAppliedLinearAccClamp=<3000 3000 3000>
    componentLinearAccScale=<0.8 0.8 0.8>
    componentLinearVelScale=<0.8 0.8 0.8>
    gravityOverride {z=-980}
    physicsBodyDefinitions [{
      sphereCollisionRadius=0
      boundBone {boneName=dyn_SkirtB01_r}
      boxExtents=<25 25 25>
      constraintSetup {angularTargetAxis=AxisY, angularLimitsMax=<5 1 60>, angularLimitsMin=<-5 -1 -15>, angularTarget=<0 0 0>}
      localJointOffset=<5 0.5 0.2>
     }]
    relativeSpaceBone {boneName=spine_05}
   }
   n17 = Anim Dynamics - Bone: dyn_SkirtB02_r(ComponentPose=@n16, Alpha=0.5, ExternalForce="0,0,0") {
    angularDampingOverride=0.999
    angularSpringConstant=800
    bAngularSpring=true
    bGravityOverrideInSimSpace=true
    bOverrideAngularDamping=true
    bUseGravityOverride=true
    gravityScale=0.2
    simulationSpace=RootRelative
    alphaBoolBlend {blendOption=Linear}
    boundBone {boneName=dyn_SkirtB02_r}
    componentAppliedLinearAccClamp=<3000 3000 3000>
    componentLinearAccScale=<0.8 0.8 0.8>
    componentLinearVelScale=<0.8 0.8 0.8>
    gravityOverride {z=-980}
    physicsBodyDefinitions [{
      sphereCollisionRadius=0
      boundBone {boneName=dyn_SkirtB02_r}
      boxExtents=<25 25 25>
      constraintSetup {angularTargetAxis=AxisY, angularLimitsMax=<5 1 60>, angularLimitsMin=<-5 -1 -15>, angularTarget=<0 0 0>}
      localJointOffset=<5 0.5 0.2>
     }]
    relativeSpaceBone {boneName=spine_05}
   }
   n18 = Anim Dynamics - Bone: dyn_SkirtC01_l(ComponentPose=@n17, Alpha=0.5, ExternalForce="0,0,0") {
    angularDampingOverride=0.8
    angularSpringConstant=800
    bAngularSpring=true
    bGravityOverrideInSimSpace=true
    bOverrideAngularDamping=true
    bUseGravityOverride=true
    gravityScale=0.2
    simulationSpace=RootRelative
    alphaBoolBlend {blendOption=Linear}
    boundBone {boneName=dyn_SkirtC01_l}
    componentAppliedLinearAccClamp=<3000 3000 3000>
    componentLinearAccScale=<0.8 0.8 0.8>
    componentLinearVelScale=<0.8 0.8 0.8>
    gravityOverride {z=-980}
    physicsBodyDefinitions [{
      sphereCollisionRadius=0
      boundBone {boneName=dyn_SkirtC01_l}
      boxExtents=<15 15 15>
      constraintSetup {angularTargetAxis=AxisY, angularLimitsMax=<5 1 45>, angularLimitsMin=<-5 -1 -15>, angularTarget=<0 0 0>}
      localJointOffset=<5 0.5 -0.2>
     }]
    relativeSpaceBone {boneName=spine_05}
   }
   n19 = Anim Dynamics - Bone: dyn_SkirtC02_l(ComponentPose=@n18, Alpha=0.5, ExternalForce="0,0,0") {
    angularDampingOverride=0.8
    angularSpringConstant=800
    bAngularSpring=true
    bGravityOverrideInSimSpace=true
    bOverrideAngularDamping=true
    bUseGravityOverride=true
    gravityScale=0.2
    simulationSpace=RootRelative
    alphaBoolBlend {blendOption=Linear}
    boundBone {boneName=dyn_SkirtC02_l}
    componentAppliedLinearAccClamp=<3000 3000 3000>
    componentLinearAccScale=<0.8 0.8 0.8>
    componentLinearVelScale=<0.8 0.8 0.8>
    gravityOverride {z=-980}
    physicsBodyDefinitions [{
      sphereCollisionRadius=0
      boundBone {boneName=dyn_SkirtC02_l}
      boxExtents=<15 15 15>
      constraintSetup {angularTargetAxis=AxisY, angularLimitsMax=<5 1 45>, angularLimitsMin=<-5 -1 -15>, angularTarget=<0 0 0>}
      localJointOffset=<5 0.5 -0.2>
     }]
    relativeSpaceBone {boneName=spine_05}
   }
   n20 = Anim Dynamics - Bone: dyn_SkirtC01_r(ComponentPose=@n19, Alpha=0.5, ExternalForce="0,0,0") {
    angularDampingOverride=0.8
    angularSpringConstant=800
    bAngularSpring=true
    bGravityOverrideInSimSpace=true
    bOverrideAngularDamping=true
    bUseGravityOverride=true
    gravityScale=0.2
    simulationSpace=RootRelative
    alphaBoolBlend {blendOption=Linear}
    boundBone {boneName=dyn_SkirtC01_r}
    componentAppliedLinearAccClamp=<3000 3000 3000>
    componentLinearAccScale=<0.8 0.8 0.8>
    componentLinearVelScale=<0.8 0.8 0.8>
    gravityOverride {z=-980}
    physicsBodyDefinitions [{
      sphereCollisionRadius=0
      boundBone {boneName=dyn_SkirtC01_r}
      boxExtents=<15 15 15>
      constraintSetup {angularTargetAxis=AxisY, angularLimitsMax=<5 1 45>, angularLimitsMin=<-5 -1 -15>, angularTarget=<0 0 0>}
      localJointOffset=<5 0.5 0.2>
     }]
    relativeSpaceBone {boneName=spine_05}
   }
   n21 = Anim Dynamics - Bone: dyn_SkirtC02_r(ComponentPose=@n20, Alpha=0.5, ExternalForce="0,0,0") {
    angularDampingOverride=0.8
    angularSpringConstant=800
    bAngularSpring=true
    bGravityOverrideInSimSpace=true
    bOverrideAngularDamping=true
    bUseGravityOverride=true
    gravityScale=0.2
    simulationSpace=RootRelative
    alphaBoolBlend {blendOption=Linear}
    boundBone {boneName=dyn_SkirtC02_r}
    componentAppliedLinearAccClamp=<3000 3000 3000>
    componentLinearAccScale=<0.8 0.8 0.8>
    componentLinearVelScale=<0.8 0.8 0.8>
    gravityOverride {z=-980}
    physicsBodyDefinitions [{
      sphereCollisionRadius=0
      boundBone {boneName=dyn_SkirtC02_r}
      boxExtents=<15 15 15>
      constraintSetup {angularTargetAxis=AxisY, angularLimitsMax=<5 1 45>, angularLimitsMin=<-5 -1 -15>, angularTarget=<0 0 0>}
      localJointOffset=<5 0.5 0.2>
     }]
    relativeSpaceBone {boneName=spine_05}
   }
   Save cached pose 'MainSetup'(Pose=Component To Local(ComponentPose=@n21)) {cacheName=MainSetup}
   n22 = RigidBody(ComponentPose=Local To Component(LocalPose=Use cached pose 'MainSetup'()), OverrideWorldGravity="0,0,10") {
    bClampLinearTranslationLimitToRefPose=true
    bOverrideWorldGravity=true
    overridePhysicsAsset=&M_MED_NightHawk_Physics
    simulationSpace=BaseBoneSpace
    alphaBoolBlend {blendOption=Linear}
    alphaScaleBiasClamp {bInterpResult=true, clampMax=0.5}
    baseBoneRef {boneName=pelvis}
    componentAppliedLinearAccClamp=<2000 2000 2000>
    componentLinearAccScale=<0.35 0.3 0.1>
    componentLinearVelScale=<0.35 0.3 0.1>
    overrideWorldGravity {z=-980}
   }
   n23 = "Transform (Modify) Bone - Bone: dyn_SkirtC02_l"(ComponentPose=@n22, Alpha=1, Translation="0,0,0", Rotation="0,15,0", Scale="1,1,1") {
    rotationMode=BMM_Additive
    rotationSpace=BCS_BoneSpace
    alphaBoolBlend {blendOption=Linear}
    alphaScaleBiasClamp {bInterpResult=true}
    boneToModify {boneName=dyn_SkirtC02_l}
    rotation {yaw=15}
   }
   n24 = "Transform (Modify) Bone - Bone: dyn_SkirtC02_r"(ComponentPose=@n23, Alpha=1, Translation="0,0,0", Rotation="0,15,0", Scale="1,1,1") {
    rotationMode=BMM_Additive
    rotationSpace=BCS_BoneSpace
    alphaBoolBlend {blendOption=Linear}
    alphaScaleBiasClamp {bInterpResult=true}
    boneToModify {boneName=dyn_SkirtC02_r}
    rotation {yaw=15}
   }
   Save cached pose 'SkydivingCache'(Pose=Component To Local(ComponentPose=@n24)) {cacheName=SkydivingCache}
   n25 = RigidBody(ComponentPose=Local To Component(LocalPose=Use cached pose 'SkydivingCache'()), Alpha=0, OverrideWorldGravity="0,0,-980") {
    bClampLinearTranslationLimitToRefPose=true
    bOverrideWorldGravity=true
    overridePhysicsAsset=&M_MED_NightHawk_Physics
    simulationSpace=BaseBoneSpace
    alphaBoolBlend {blendOption=Linear}
    alphaScaleBiasClamp {bInterpResult=true}
    baseBoneRef {boneName=spine_02}
    componentAppliedLinearAccClamp=<2000 2000 2000>
    componentLinearAccScale=<0.35 0.3 0.1>
    componentLinearVelScale=<0.35 0.3 0.1>
    overrideWorldGravity {z=-980}
   }
   n26 = RigidBody(ComponentPose=Local To Component(LocalPose=Use cached pose 'SkydivingCache'()), Alpha=0.3, OverrideWorldGravity="0,0,-980") {
    bClampLinearTranslationLimitToRefPose=true
    bOverrideWorldGravity=true
    overridePhysicsAsset=&M_MED_NightHawk_Physics
    simulationSpace=BaseBoneSpace
    alphaBoolBlend {blendOption=Linear}
    alphaScaleBiasClamp {bInterpResult=true}
    baseBoneRef {boneName=spine_02}
    componentAppliedLinearAccClamp=<2000 2000 2000>
    componentLinearAccScale=<0.35 0.3 0.1>
    componentLinearVelScale=<0.35 0.3 0.1>
    overrideWorldGravity {z=-980}
   }
   Save cached pose 'PostEmoteCache'(Pose=Blend Poses by bool(
     BlendPose_0=Component To Local(ComponentPose=@n26)
     BlendPose_1=Component To Local(ComponentPose=@n25)
     BlendTime_0=0.5
     BlendTime_1=0.5
     bActiveValue=Get ClearRadiusEmotePlaying() -> (ClearRadiusEmotePlaying:bool)
    ) {blendType=Linear}) {cacheName=PostEmoteCache}
   n27 = "Constraint - attach_backpack (PPPPPP)"!(
    ComponentPose=Local To Component(LocalPose=Copy Pose From Mesh() {bUseAttachedParent=true})
    Alpha=1
    ConstraintWeights_0=0.7517
    ConstraintWeights_1:float=0.1526
    ConstraintWeights_2:float=0.0579
    ConstraintWeights_3:float=0.0228
    ConstraintWeights_4:float=0.0141
    ConstraintWeights_5:float=0.0009
   ) -> (Pose) {
    alphaBoolBlend {blendOption=Linear}
    boneToModify {boneName=attach_backpack}
    constraintSetup [
     {targetBone {boneName=dyn_hat_bk_02}}
     {targetBone {boneName=dyn_hat_bk_01}}
     {targetBone {boneName=spine_04}}
     {targetBone {boneName=spine_05}}
     {targetBone {boneName=spine_03}}
     {targetBone {boneName=clavicle_r}}
    ]
    constraintWeights [1,1,1,1,1,1]
   }
   n28 = "Constraint - attach_cape (PPPPPP)"!(
    ComponentPose=@n27
    Alpha=1
    ConstraintWeights_0=0.522
    ConstraintWeights_1:float=0.3623
    ConstraintWeights_2:float=0.0614
    ConstraintWeights_3:float=0.0518
    ConstraintWeights_4:float=0.0017
    ConstraintWeights_5:float=0.0009
   ) -> (Pose) {
    alphaBoolBlend {blendOption=Linear}
    boneToModify {boneName=attach_cape}
    constraintSetup [
     {targetBone {boneName=spine_05}}
     {targetBone {boneName=dyn_hat_bk_01}}
     {targetBone {boneName=neck_01}}
     {targetBone {boneName=spine_04}}
     {targetBone {boneName=clavicle_r}}
     {targetBone {boneName=clavicle_l}}
    ]
    constraintWeights [1,1,1,1,1,1]
   }
   n29 = Get CapeHide() -> (CapeHide:double)
   n30 = "Transform (Modify) Bone - Bone: attach_backpack"(ComponentPose=@n28, Alpha=@n29, Translation="0,6,0", Rotation="0,0,8", Scale="1,1,1") {
    rotationMode=BMM_Additive
    rotationSpace=BCS_BoneSpace
    scaleMode=BMM_Additive
    scaleSpace=BCS_BoneSpace
    translationMode=BMM_Additive
    translationSpace=BCS_BoneSpace
    alphaBoolBlend {blendOption=Linear}
    boneToModify {boneName=attach_backpack}
    rotation {roll=8}
    translation {y=6}
   }
   n31 = "Transform (Modify) Bone - Bone: attach_cape"(ComponentPose=@n30, Alpha=@n29, Translation="0,4,2", Rotation="0,0,5", Scale="1,1,1") {
    rotationMode=BMM_Additive
    rotationSpace=BCS_BoneSpace
    scaleMode=BMM_Additive
    scaleSpace=BCS_BoneSpace
    translationMode=BMM_Additive
    translationSpace=BCS_BoneSpace
    alphaBoolBlend {blendOption=Linear}
    boneToModify {boneName=attach_cape}
    rotation {roll=5}
    translation {y=4, z=2}
   }
   n32 = Get neck_rot_up() -> (neck_rot_up:double)
   n33 = "Constraint - dyn_collar_01 (RR)"!(ComponentPose=@n31, Alpha=@n32, ConstraintWeights_0=0.8, ConstraintWeights_1:float=0.2) -> (Pose) {
    alphaBoolBlend {blendOption=Linear}
    boneToModify {boneName=dyn_collar_01}
    constraintSetup [{transformType=Rotation, targetBone {boneName=spine_05}}, {transformType=Rotation, targetBone {boneName=head}}]
    constraintWeights [1,1]
   }
   n34 = "Constraint - dyn_collarB_01_r (RR)"!(ComponentPose=@n33, Alpha=@n32, ConstraintWeights_0=0.8, ConstraintWeights_1:float=0.2) -> (Pose) {
    alphaBoolBlend {blendOption=Linear}
    boneToModify {boneName=dyn_collarB_01_r}
    constraintSetup [{transformType=Rotation, targetBone {boneName=spine_05}}, {transformType=Rotation, targetBone {boneName=head}}]
    constraintWeights [1,1]
   }
   n35 = "Transform (Modify) Bone - Bone: dyn_collar_01"(ComponentPose=@n34, Alpha=@n32, Translation="-5,-2,0", Rotation="0,-60,0", Scale="1,1,1") {
    rotationMode=BMM_Additive
    rotationSpace=BCS_BoneSpace
    scaleMode=BMM_Additive
    scaleSpace=BCS_BoneSpace
    translationMode=BMM_Additive
    translationSpace=BCS_BoneSpace
    alphaBoolBlend {blendOption=Linear}
    boneToModify {boneName=dyn_collar_01}
    rotation {yaw=-60}
    translation=<-5 -2>
   }
   n36 = "Transform (Modify) Bone - Bone: dyn_collarB_01_r"(ComponentPose=@n35, Alpha=@n32, Translation="0,0,0", Rotation="20,20,0", Scale="1,1,1") {
    rotationMode=BMM_Additive
    rotationSpace=BCS_BoneSpace
    scaleMode=BMM_Additive
    scaleSpace=BCS_BoneSpace
    translationMode=BMM_Additive
    translationSpace=BCS_BoneSpace
    alphaBoolBlend {blendOption=Linear}
    boneToModify {boneName=dyn_collarB_01_r}
    rotation {pitch=20, yaw=20}
   }
   n37 = "Transform (Modify) Bone - Bone: dyn_collarB_01_r"(ComponentPose=@n36, Alpha=Get neck_rot_r() -> (neck_rot_r:double), Translation="0,0,0", Rotation="50,10,-40", Scale="1,1,1") {
    rotationMode=BMM_Additive
    rotationSpace=BCS_BoneSpace
    scaleMode=BMM_Additive
    scaleSpace=BCS_BoneSpace
    translationMode=BMM_Additive
    translationSpace=BCS_BoneSpace
    alphaBoolBlend {blendOption=Linear}
    boneToModify {boneName=dyn_collarB_01_r}
    rotation=<rot 50 10 -40>
   }
   n38 = "Transform (Modify) Bone - Bone: dyn_collarB_01_r"(ComponentPose=@n37, Alpha=@n32, Translation="-2,-2,0", Rotation="48.04,10,4.36", Scale="1,1,1") {
    rotationMode=BMM_Additive
    rotationSpace=BCS_BoneSpace
    scaleMode=BMM_Additive
    scaleSpace=BCS_BoneSpace
    translationMode=BMM_Additive
    translationSpace=BCS_BoneSpace
    alphaBoolBlend {blendOption=Linear}
    boneToModify {boneName=dyn_collarB_01_r}
    rotation=<rot 48.04 10 4.36>
    translation=<-2 -2>
   }
   n39 = "Transform (Modify) Bone - Bone: dyn_collarB_01_l"(ComponentPose=@n38, Alpha=@n32, Translation="0,0,0", Rotation="11.446,0.525,0.349", Scale="1,1,1") {
    rotationMode=BMM_Additive
    rotationSpace=BCS_BoneSpace
    scaleMode=BMM_Additive
    scaleSpace=BCS_BoneSpace
    translationMode=BMM_Additive
    translationSpace=BCS_BoneSpace
    alphaBoolBlend {blendOption=Linear}
    boneToModify {boneName=dyn_collarB_01_l}
    rotation=<rot 11.446 0.525 0.349>
   }
   n40 = Anim Dynamics - Bone: dyn_cravat_01(ComponentPose=@n39, Alpha=0.5, ExternalForce="0,0,0") {
    angularDampingOverride=0.999
    angularSpringConstant=800
    bAngularSpring=true
    bGravityOverrideInSimSpace=true
    bOverrideAngularDamping=true
    bUseGravityOverride=true
    gravityScale=0.2
    simulationSpace=RootRelative
    alphaBoolBlend {blendOption=Linear}
    boundBone {boneName=dyn_cravat_01}
    componentAppliedLinearAccClamp=<6000 6000 6000>
    componentLinearAccScale=<0.8 0.8 0.8>
    componentLinearVelScale=<0.8 0.8 0.8>
    gravityOverride {z=-980}
    physicsBodyDefinitions [{
      sphereCollisionRadius=0
      boundBone {boneName=dyn_cravat_01}
      boxExtents=<15 15 15>
      constraintSetup {angularTargetAxis=AxisY, angularLimitsMax=<5 45 5>, angularLimitsMin=<-5 -5 -5>, angularTarget=<0 0 0>}
      localJointOffset=<5 0.1 1>
     }]
    relativeSpaceBone {boneName=spine_05}
   }
   n41 = Anim Dynamics - Bone: dyn_cravat_02(ComponentPose=@n40, Alpha=0.5, ExternalForce="0,0,0") {
    angularDampingOverride=0.999
    angularSpringConstant=800
    bAngularSpring=true
    bGravityOverrideInSimSpace=true
    bOverrideAngularDamping=true
    bUseGravityOverride=true
    gravityScale=0.2
    simulationSpace=RootRelative
    alphaBoolBlend {blendOption=Linear}
    boundBone {boneName=dyn_cravat_02}
    componentAppliedLinearAccClamp=<6000 6000 6000>
    componentLinearAccScale=<0.8 0.8 0.8>
    componentLinearVelScale=<0.8 0.8 0.8>
    gravityOverride {z=-980}
    physicsBodyDefinitions [{
      sphereCollisionRadius=0
      boundBone {boneName=dyn_cravat_02}
      boxExtents=<15 15 15>
      constraintSetup {angularTargetAxis=AxisY, angularLimitsMax=<5 45 5>, angularLimitsMin=<-5 -5 -5>, angularTarget=<0 0 0>}
      localJointOffset=<5 0.1 1>
     }]
    relativeSpaceBone {boneName=spine_05}
   }
   n42 = "Constraint - thigh_volume_bk_l (PP)"!(ComponentPose=@n41, Alpha=1, ConstraintWeights_0=0, ConstraintWeights_1:float=1) -> (Pose) {
    alphaBoolBlend {blendOption=Linear}
    boneToModify {boneName=thigh_volume_bk_l}
    constraintSetup [{targetBone {boneName=thigh_twist_01_l}}, {targetBone {boneName=pelvis}}]
    constraintWeights [1,1]
   }
   n43 = "Constraint - thigh_volume_fn_l (RR)"!(ComponentPose=@n42, Alpha=Get thigh_shape_fn_l() -> (thigh_shape_fn_l:double), ConstraintWeights_0=0.5, ConstraintWeights_1:float=0.5) -> (Pose) {
    alphaBoolBlend {blendOption=Linear}
    boneToModify {boneName=thigh_volume_fn_l}
    constraintSetup [{transformType=Rotation, targetBone {boneName=thigh_twist_01_l}}, {transformType=Rotation, targetBone {boneName=pelvis}}]
    constraintWeights [1,1]
   }
   n44 = "Constraint - thigh_volume_lt_l (PP)"!(ComponentPose=@n43, Alpha=1, ConstraintWeights_0=0.7, ConstraintWeights_1:float=0.3) -> (Pose) {
    alphaBoolBlend {blendOption=Linear}
    boneToModify {boneName=thigh_volume_lt_l}
    constraintSetup [{targetBone {boneName=thigh_twist_01_l}}, {targetBone {boneName=pelvis}}]
    constraintWeights [1,1]
   }
   n45 = "Constraint - thigh_volume_bk_r (PP)"!(ComponentPose=@n44, Alpha=1, ConstraintWeights_0=0, ConstraintWeights_1:float=1) -> (Pose) {
    alphaBoolBlend {blendOption=Linear}
    boneToModify {boneName=thigh_volume_bk_r}
    constraintSetup [{targetBone {boneName=thigh_twist_01_r}}, {targetBone {boneName=pelvis}}]
    constraintWeights [1,1]
   }
   n46 = "Constraint - thigh_volume_fn_r (RR)"!(ComponentPose=@n45, Alpha=Get thigh_shape_fn_r() -> (thigh_shape_fn_r:double), ConstraintWeights_0=0.5, ConstraintWeights_1:float=0.5) -> (Pose) {
    alphaBoolBlend {blendOption=Linear}
    boneToModify {boneName=thigh_volume_fn_r}
    constraintSetup [{transformType=Rotation, targetBone {boneName=thigh_twist_01_r}}, {transformType=Rotation, targetBone {boneName=pelvis}}]
    constraintWeights [1,1]
   }
   n47 = "Constraint - thigh_volume_rt_r (PP)"!(ComponentPose=@n46, Alpha=1, ConstraintWeights_0=0.7, ConstraintWeights_1:float=0.3) -> (Pose) {
    alphaBoolBlend {blendOption=Linear}
    boneToModify {boneName=thigh_volume_rt_r}
    constraintSetup [{targetBone {boneName=thigh_twist_01_r}}, {targetBone {boneName=pelvis}}]
    constraintWeights [1,1]
   }
   n48 = "Transform (Modify) Bone - Bone: thigh_volume_fn_l"(ComponentPose=@n47, Alpha=Get thigh_shape_fn_l(), Translation="-5,-10,0", Rotation="-10,25,0", Scale="1,1,1") {
    rotationMode=BMM_Additive
    rotationSpace=BCS_BoneSpace
    scaleMode=BMM_Additive
    scaleSpace=BCS_BoneSpace
    translationMode=BMM_Additive
    translationSpace=BCS_BoneSpace
    alphaBoolBlend {blendOption=Linear}
    boneToModify {boneName=thigh_volume_fn_l}
    rotation {pitch=-10, yaw=25}
    translation=<-5 -10>
   }
   n49 = "Transform (Modify) Bone - Bone: thigh_volume_fn_r"(ComponentPose=@n48, Alpha=Get thigh_shape_fn_r(), Translation="5,-10,0", Rotation="-10,-25,0", Scale="1,1,1") {
    rotationMode=BMM_Additive
    rotationSpace=BCS_BoneSpace
    scaleMode=BMM_Additive
    scaleSpace=BCS_BoneSpace
    translationMode=BMM_Additive
    translationSpace=BCS_BoneSpace
    alphaBoolBlend {blendOption=Linear}
    boneToModify {boneName=thigh_volume_fn_r}
    rotation {pitch=-10, yaw=-25}
    translation=<5 -10>
   }
   n50 = "Transform (Modify) Bone - Bone: thigh_volume_bk_l"(ComponentPose=@n49, Alpha=Get thigh_shape_bk_l(), Translation="-15,-1,0", Rotation="0,0,0", Scale="1.5,1.5,1.5") {
    rotationMode=BMM_Additive
    rotationSpace=BCS_BoneSpace
    scaleMode=BMM_Additive
    scaleSpace=BCS_BoneSpace
    translationMode=BMM_Additive
    translationSpace=BCS_BoneSpace
    alphaBoolBlend {blendOption=Linear}
    boneToModify {boneName=thigh_volume_bk_l}
    scale=<1.5 1.5 1.5>
    translation=<-15 -1>
   }
   n51 = "Transform (Modify) Bone - Bone: thigh_volume_bk_r"(ComponentPose=@n50, Alpha=Get thigh_shape_bk_r(), Translation="15,-1,0", Rotation="0,0,0", Scale="1.5,1.5,1.5") {
    rotationMode=BMM_Additive
    rotationSpace=BCS_BoneSpace
    scaleMode=BMM_Additive
    scaleSpace=BCS_BoneSpace
    translationMode=BMM_Additive
    translationSpace=BCS_BoneSpace
    alphaBoolBlend {blendOption=Linear}
    boneToModify {boneName=thigh_volume_bk_r}
    scale=<1.5 1.5 1.5>
    translation=<15 -1>
   }
   n52 = "Transform (Modify) Bone - Bone: thigh_volume_lt_l"(ComponentPose=@n51, Alpha=Get thigh_shape_lt_l() -> (thigh_shape_lt_l:double), Translation="0,-1,0", Rotation="0,0,0", Scale="1,1,1") {
    rotationMode=BMM_Additive
    rotationSpace=BCS_BoneSpace
    scaleMode=BMM_Additive
    scaleSpace=BCS_BoneSpace
    translationMode=BMM_Additive
    translationSpace=BCS_BoneSpace
    alphaBoolBlend {blendOption=Linear}
    boneToModify {boneName=thigh_volume_lt_l}
    translation {y=-1}
   }
   n53 = "Transform (Modify) Bone - Bone: thigh_volume_rt_r"(ComponentPose=@n52, Alpha=Get thigh_shape_rt_r() -> (thigh_shape_rt_r:double), Translation="0,-3,0", Rotation="0,-10,0", Scale="1,1,1") {
    rotationMode=BMM_Additive
    rotationSpace=BCS_BoneSpace
    scaleMode=BMM_Additive
    scaleSpace=BCS_BoneSpace
    translationMode=BMM_Additive
    translationSpace=BCS_BoneSpace
    alphaBoolBlend {blendOption=Linear}
    boneToModify {boneName=thigh_volume_rt_r}
    rotation {yaw=-10}
    translation {y=-3}
   }
   n54 = Anim Dynamics - Bone: dyn_bootA_01_l(ComponentPose=@n53, Alpha=0.5, ExternalForce="0,0,0") {
    angularDampingOverride=1
    angularSpringConstant=800
    bAngularSpring=true
    bGravityOverrideInSimSpace=true
    bOverrideAngularDamping=true
    bOverrideLinearDamping=true
    bUseGravityOverride=true
    gravityScale=0.2
    linearDampingOverride=0.5
    simulationSpace=RootRelative
    alphaBoolBlend {blendOption=Linear}
    boundBone {boneName=dyn_bootA_01_l}
    componentAppliedLinearAccClamp=<4000 4000 4000>
    componentLinearAccScale=<0.6 0.6 0.6>
    componentLinearVelScale=<0.6 0.6 0.6>
    gravityOverride {z=-980}
    physicsBodyDefinitions [{
      sphereCollisionRadius=0
      boundBone {boneName=dyn_bootA_01_l}
      boxExtents=<35 35 35>
      constraintSetup {angularTargetAxis=AxisY, angularLimitsMax=<5 1 15>, angularLimitsMin=<-5 -1 -30>, angularTarget=<0 0 0>}
      localJointOffset=<5 -0.5 -0.2>
     }]
    relativeSpaceBone {boneName=spine_05}
   }
   n55 = Anim Dynamics - Bone: dyn_bootA_rope01_l(ComponentPose=@n54, Alpha=0.5, ExternalForce="0,0,0") {
    angularDampingOverride=1
    angularSpringConstant=800
    bAngularSpring=true
    bGravityOverrideInSimSpace=true
    bOverrideAngularDamping=true
    bUseGravityOverride=true
    gravityScale=0.2
    linearDampingOverride=0.5
    simulationSpace=RootRelative
    alphaBoolBlend {blendOption=Linear}
    boundBone {boneName=dyn_bootA_rope01_l}
    componentAppliedLinearAccClamp=<6000 6000 6000>
    componentLinearAccScale=<0.6 0.6 0.6>
    componentLinearVelScale=<0.6 0.6 0.6>
    gravityOverride {z=-980}
    physicsBodyDefinitions [{
      sphereCollisionRadius=0
      boundBone {boneName=dyn_bootA_rope01_l}
      boxExtents=<10 10 10>
      constraintSetup {angularTargetAxis=AxisY, angularLimitsMax=<5 30 40>, angularLimitsMin=<-5 -30 -5>, angularTarget=<0 0 0>}
      localJointOffset=<4 0.5 1>
     }]
    relativeSpaceBone {boneName=spine_05}
   }
   n56 = Anim Dynamics - Bone: dyn_bootA_rope02_l(ComponentPose=@n55, Alpha=0.5, ExternalForce="0,0,0") {
    angularDampingOverride=1
    angularSpringConstant=800
    bAngularSpring=true
    bGravityOverrideInSimSpace=true
    bOverrideAngularDamping=true
    bUseGravityOverride=true
    gravityScale=0.2
    linearDampingOverride=0.5
    simulationSpace=RootRelative
    alphaBoolBlend {blendOption=Linear}
    boundBone {boneName=dyn_bootA_rope02_l}
    componentAppliedLinearAccClamp=<6000 6000 6000>
    componentLinearAccScale=<0.6 0.6 0.6>
    componentLinearVelScale=<0.6 0.6 0.6>
    gravityOverride {z=-980}
    physicsBodyDefinitions [{
      sphereCollisionRadius=0
      boundBone {boneName=dyn_bootA_rope02_l}
      boxExtents=<10 10 10>
      constraintSetup {angularTargetAxis=AxisY, angularLimitsMax=<5 30 15>, angularLimitsMin=<-5 -30 -40>, angularTarget=<0 0 0>}
      localJointOffset=<4 0.5 1>
     }]
    relativeSpaceBone {boneName=spine_05}
   }
   n57 = Anim Dynamics - Bone: dyn_bootB_01_l(ComponentPose=@n56, Alpha=0.5, ExternalForce="0,0,0") {
    angularDampingOverride=1
    angularSpringConstant=800
    bAngularSpring=true
    bGravityOverrideInSimSpace=true
    bOverrideAngularDamping=true
    bOverrideLinearDamping=true
    bUseGravityOverride=true
    gravityScale=0.2
    linearDampingOverride=0.5
    simulationSpace=RootRelative
    alphaBoolBlend {blendOption=Linear}
    boundBone {boneName=dyn_bootB_01_l}
    componentAppliedLinearAccClamp=<4000 4000 4000>
    componentLinearAccScale=<0.6 0.6 0.6>
    componentLinearVelScale=<0.6 0.6 0.6>
    gravityOverride {z=-980}
    physicsBodyDefinitions [{
      sphereCollisionRadius=0
      boundBone {boneName=dyn_bootB_01_l}
      boxExtents=<25 25 25>
      constraintSetup {angularTargetAxis=AxisY, angularLimitsMax=<5 1 60>, angularLimitsMin=<-5 -1 -25>, angularTarget=<0 0 0>}
      localJointOffset=<5 0.5 -0.2>
     }]
    relativeSpaceBone {boneName=spine_05}
   }
   n58 = Anim Dynamics - Bone: dyn_bootC_01_l(ComponentPose=@n57, Alpha=0.5, ExternalForce="0,0,0") {
    angularDampingOverride=1
    angularSpringConstant=800
    bAngularSpring=true
    bGravityOverrideInSimSpace=true
    bOverrideAngularDamping=true
    bOverrideLinearDamping=true
    bUseGravityOverride=true
    gravityScale=0.2
    linearDampingOverride=0.5
    simulationSpace=RootRelative
    alphaBoolBlend {blendOption=Linear}
    boundBone {boneName=dyn_bootC_01_l}
    componentAppliedLinearAccClamp=<4000 4000 4000>
    componentLinearAccScale=<0.6 0.6 0.6>
    componentLinearVelScale=<0.6 0.6 0.6>
    gravityOverride {z=-980}
    physicsBodyDefinitions [{
      sphereCollisionRadius=0
      boundBone {boneName=dyn_bootC_01_l}
      boxExtents=<35 35 35>
      constraintSetup {angularTargetAxis=AxisY, angularLimitsMax=<5 1 60>, angularLimitsMin=<-5 -1 -25>, angularTarget=<0 0 0>}
      localJointOffset=<5 0.5 -0.2>
     }]
    relativeSpaceBone {boneName=spine_05}
   }
   n59 = Anim Dynamics - Bone: dyn_bootC_rope01_l(ComponentPose=@n58, Alpha=0.5, ExternalForce="0,0,0") {
    angularDampingOverride=1
    angularSpringConstant=800
    bAngularSpring=true
    bGravityOverrideInSimSpace=true
    bOverrideAngularDamping=true
    bUseGravityOverride=true
    gravityScale=0.2
    linearDampingOverride=0.5
    simulationSpace=RootRelative
    alphaBoolBlend {blendOption=Linear}
    boundBone {boneName=dyn_bootC_rope01_l}
    componentAppliedLinearAccClamp=<6000 6000 6000>
    componentLinearAccScale=<0.6 0.6 0.6>
    componentLinearVelScale=<0.6 0.6 0.6>
    gravityOverride {z=-980}
    physicsBodyDefinitions [{
      sphereCollisionRadius=0
      boundBone {boneName=dyn_bootC_rope01_l}
      boxExtents=<15 15 15>
      constraintSetup {angularTargetAxis=AxisY, angularLimitsMax=<5 30 60>, angularLimitsMin=<-5 -30 -40>, angularTarget=<0 0 0>}
      localJointOffset=<5 0.5 1>
     }]
    relativeSpaceBone {boneName=spine_05}
   }
   n60 = Anim Dynamics - Bone: dyn_bootC_rope02_l(ComponentPose=@n59, Alpha=0.5, ExternalForce="0,0,0") {
    angularDampingOverride=1
    angularSpringConstant=800
    bAngularSpring=true
    bGravityOverrideInSimSpace=true
    bOverrideAngularDamping=true
    bUseGravityOverride=true
    gravityScale=0.2
    linearDampingOverride=0.5
    simulationSpace=RootRelative
    alphaBoolBlend {blendOption=Linear}
    boundBone {boneName=dyn_bootC_rope02_l}
    componentAppliedLinearAccClamp=<6000 6000 6000>
    componentLinearAccScale=<0.6 0.6 0.6>
    componentLinearVelScale=<0.6 0.6 0.6>
    gravityOverride {z=-980}
    physicsBodyDefinitions [{
      sphereCollisionRadius=0
      boundBone {boneName=dyn_bootC_rope02_l}
      boxExtents=<15 15 15>
      constraintSetup {angularTargetAxis=AxisY, angularLimitsMax=<5 30 60>, angularLimitsMin=<-5 -30 -40>, angularTarget=<0 0 0>}
      localJointOffset=<5 0.5 1>
     }]
    relativeSpaceBone {boneName=spine_05}
   }
   n61 = Anim Dynamics - Bone: dyn_bootA_01_r(ComponentPose=@n60, Alpha=0.5, ExternalForce="0,0,0") {
    angularDampingOverride=1
    angularSpringConstant=800
    bAngularSpring=true
    bGravityOverrideInSimSpace=true
    bOverrideAngularDamping=true
    bOverrideLinearDamping=true
    bUseGravityOverride=true
    gravityScale=0.2
    linearDampingOverride=0.5
    simulationSpace=RootRelative
    alphaBoolBlend {blendOption=Linear}
    boundBone {boneName=dyn_bootA_01_r}
    componentAppliedLinearAccClamp=<4000 4000 4000>
    componentLinearAccScale=<0.6 0.6 0.6>
    componentLinearVelScale=<0.6 0.6 0.6>
    gravityOverride {z=-980}
    physicsBodyDefinitions [{
      sphereCollisionRadius=0
      boundBone {boneName=dyn_bootA_01_r}
      boxExtents=<35 35 35>
      constraintSetup {angularTargetAxis=AxisY, angularLimitsMax=<5 1 5>, angularLimitsMin=<-5 -1 -40>, angularTarget=<0 0 0>}
      localJointOffset=<-5 0.5 -0.2>
     }]
    relativeSpaceBone {boneName=spine_05}
   }
   n62 = Anim Dynamics - Bone: dyn_bootA_rope01_r(ComponentPose=@n61, Alpha=0.5, ExternalForce="0,0,0") {
    angularDampingOverride=1
    angularSpringConstant=800
    bAngularSpring=true
    bGravityOverrideInSimSpace=true
    bOverrideAngularDamping=true
    bUseGravityOverride=true
    gravityScale=0.2
    linearDampingOverride=0.5
    simulationSpace=RootRelative
    alphaBoolBlend {blendOption=Linear}
    boundBone {boneName=dyn_bootA_rope01_r}
    componentAppliedLinearAccClamp=<6000 6000 6000>
    componentLinearAccScale=<0.6 0.6 0.6>
    componentLinearVelScale=<0.6 0.6 0.6>
    gravityOverride {z=-980}
    physicsBodyDefinitions [{
      sphereCollisionRadius=0
      boundBone {boneName=dyn_bootA_rope01_r}
      boxExtents=<12 12 12>
      constraintSetup {angularTargetAxis=AxisY, angularLimitsMax=<5 30 60>, angularLimitsMin=<-5 -30 0>, angularTarget=<0 0 0>}
      localJointOffset=<-5 -1 -0.2>
     }]
    relativeSpaceBone {boneName=spine_05}
   }
   n63 = Anim Dynamics - Bone: dyn_bootA_rope02_r(ComponentPose=@n62, Alpha=0.5, ExternalForce="0,0,0") {
    angularDampingOverride=1
    angularSpringConstant=800
    bAngularSpring=true
    bGravityOverrideInSimSpace=true
    bOverrideAngularDamping=true
    bUseGravityOverride=true
    gravityScale=0.2
    linearDampingOverride=0.5
    simulationSpace=RootRelative
    alphaBoolBlend {blendOption=Linear}
    boundBone {boneName=dyn_bootA_rope02_r}
    componentAppliedLinearAccClamp=<6000 6000 6000>
    componentLinearAccScale=<0.6 0.6 0.6>
    componentLinearVelScale=<0.6 0.6 0.6>
    gravityOverride {z=-980}
    physicsBodyDefinitions [{
      sphereCollisionRadius=0
      boundBone {boneName=dyn_bootA_rope02_r}
      boxExtents=<12 12 12>
      constraintSetup {angularTargetAxis=AxisY, angularLimitsMax=<5 30 60>, angularLimitsMin=<-5 -30 -10>, angularTarget=<0 0 0>}
      localJointOffset=<-5 -1 -0.2>
     }]
    relativeSpaceBone {boneName=spine_05}
   }
   n64 = Anim Dynamics - Bone: dyn_bootB_01_r(ComponentPose=@n63, Alpha=0.5, ExternalForce="0,0,0") {
    angularDampingOverride=1
    angularSpringConstant=800
    bAngularSpring=true
    bGravityOverrideInSimSpace=true
    bOverrideAngularDamping=true
    bOverrideLinearDamping=true
    bUseGravityOverride=true
    gravityScale=0.2
    linearDampingOverride=0.5
    simulationSpace=RootRelative
    alphaBoolBlend {blendOption=Linear}
    boundBone {boneName=dyn_bootB_01_r}
    componentAppliedLinearAccClamp=<4000 4000 4000>
    componentLinearAccScale=<0.6 0.6 0.6>
    componentLinearVelScale=<0.6 0.6 0.6>
    gravityOverride {z=-980}
    physicsBodyDefinitions [{
      sphereCollisionRadius=0
      boundBone {boneName=dyn_bootB_01_r}
      boxExtents=<25 25 25>
      constraintSetup {angularTargetAxis=AxisY, angularLimitsMax=<5 1 60>, angularLimitsMin=<-5 -1 -25>, angularTarget=<0 0 0>}
      localJointOffset=<-5 0.5 -0.2>
     }]
    relativeSpaceBone {boneName=spine_05}
   }
   n65 = Anim Dynamics - Bone: dyn_bootC_01_r(ComponentPose=@n64, Alpha=0.5, ExternalForce="0,0,0") {
    angularDampingOverride=1
    angularSpringConstant=800
    bAngularSpring=true
    bGravityOverrideInSimSpace=true
    bOverrideAngularDamping=true
    bOverrideLinearDamping=true
    bUseGravityOverride=true
    gravityScale=0.2
    linearDampingOverride=0.5
    simulationSpace=RootRelative
    alphaBoolBlend {blendOption=Linear}
    boundBone {boneName=dyn_bootC_01_r}
    componentAppliedLinearAccClamp=<4000 4000 4000>
    componentLinearAccScale=<0.6 0.6 0.6>
    componentLinearVelScale=<0.6 0.6 0.6>
    gravityOverride {z=-980}
    physicsBodyDefinitions [{
      sphereCollisionRadius=0
      boundBone {boneName=dyn_bootC_01_r}
      boxExtents=<35 35 35>
      constraintSetup {angularTargetAxis=AxisY, angularLimitsMax=<5 1 60>, angularLimitsMin=<-5 -1 -25>, angularTarget=<0 0 0>}
      localJointOffset=<-5 0.5 -0.2>
     }]
    relativeSpaceBone {boneName=spine_05}
   }
   n66 = Anim Dynamics - Bone: dyn_bootC_rope01_r(ComponentPose=@n65, Alpha=0.5, ExternalForce="0,0,0") {
    angularDampingOverride=1
    angularSpringConstant=800
    bAngularSpring=true
    bGravityOverrideInSimSpace=true
    bOverrideAngularDamping=true
    bUseGravityOverride=true
    gravityScale=0.2
    linearDampingOverride=0.5
    simulationSpace=RootRelative
    alphaBoolBlend {blendOption=Linear}
    boundBone {boneName=dyn_bootC_rope01_r}
    componentAppliedLinearAccClamp=<6000 6000 6000>
    componentLinearAccScale=<0.6 0.6 0.6>
    componentLinearVelScale=<0.6 0.6 0.6>
    gravityOverride {z=-980}
    physicsBodyDefinitions [{
      sphereCollisionRadius=0
      boundBone {boneName=dyn_bootC_rope01_r}
      boxExtents=<12 12 12>
      constraintSetup {angularTargetAxis=AxisY, angularLimitsMax=<5 30 60>, angularLimitsMin=<-5 -30 -40>, angularTarget=<0 0 0>}
      localJointOffset=<-5 0.5 -0.2>
     }]
    relativeSpaceBone {boneName=spine_05}
   }
   n67 = Anim Dynamics - Bone: dyn_bootC_rope02_r(ComponentPose=@n66, Alpha=0.5, ExternalForce="0,0,0") {
    angularDampingOverride=1
    angularSpringConstant=800
    bAngularSpring=true
    bGravityOverrideInSimSpace=true
    bOverrideAngularDamping=true
    bUseGravityOverride=true
    gravityScale=0.2
    linearDampingOverride=0.5
    simulationSpace=RootRelative
    alphaBoolBlend {blendOption=Linear}
    boundBone {boneName=dyn_bootC_rope02_r}
    componentAppliedLinearAccClamp=<6000 6000 6000>
    componentLinearAccScale=<0.6 0.6 0.6>
    componentLinearVelScale=<0.6 0.6 0.6>
    gravityOverride {z=-980}
    physicsBodyDefinitions [{
      sphereCollisionRadius=0
      boundBone {boneName=dyn_bootC_rope02_r}
      boxExtents=<12 12 12>
      constraintSetup {angularTargetAxis=AxisY, angularLimitsMax=<5 30 60>, angularLimitsMin=<-5 -30 -40>, angularTarget=<0 0 0>}
      localJointOffset=<-5 0.5 -0.2>
     }]
    relativeSpaceBone {boneName=spine_05}
   }
   n68 = "Constraint - calf_r_front (RR)"!(ComponentPose=@n67, Alpha=1, ConstraintWeights_0=0.5, ConstraintWeights_1:float=0.5) -> (Pose) {
    alphaBoolBlend {blendOption=Linear}
    boneToModify {boneName=calf_r_front}
    constraintSetup [{transformType=Rotation, targetBone {boneName=thigh_l}}, {transformType=Rotation, targetBone {boneName=calf_l}}]
    constraintWeights [1,1]
   }
   n69 = "Constraint - calf_r_front (RR)"(ComponentPose=@n68, Alpha=1, ConstraintWeights_0=0.5, ConstraintWeights_1=0.5) {
    alphaBoolBlend {blendOption=Linear}
    boneToModify {boneName=calf_r_front}
    constraintSetup [{transformType=Rotation, targetBone {boneName=thigh_r}}, {transformType=Rotation, targetBone {boneName=calf_r}}]
    constraintWeights [1,1]
   }
   n70 = "Constraint - calf_l_back (RR)"!(ComponentPose=@n69, Alpha=1, ConstraintWeights_0=0.5, ConstraintWeights_1:float=0.5) -> (Pose) {
    alphaBoolBlend {blendOption=Linear}
    boneToModify {boneName=calf_l_back}
    constraintSetup [{transformType=Rotation, targetBone {boneName=thigh_l}}, {transformType=Rotation, targetBone {boneName=calf_l}}]
    constraintWeights [1,1]
   }
   n71 = "Constraint - calf_r_back (RR)"!(ComponentPose=@n70, Alpha=1, ConstraintWeights_0=0.5, ConstraintWeights_1:float=0.5) -> (Pose) {
    alphaBoolBlend {blendOption=Linear}
    boneToModify {boneName=calf_r_back}
    constraintSetup [{transformType=Rotation, targetBone {boneName=thigh_r}}, {transformType=Rotation, targetBone {boneName=calf_r}}]
    constraintWeights [1,1]
   }
   n72 = Get calf_l_front() -> (calf_l_front:double)
   n73 = "Transform (Modify) Bone - Bone: calf_l_front"(ComponentPose=@n71, Alpha=@n72, Translation="7,-5,0", Rotation="0,0,0", Scale="2,2,2") {
    rotationMode=BMM_Additive
    rotationSpace=BCS_BoneSpace
    scaleMode=BMM_Additive
    scaleSpace=BCS_BoneSpace
    translationMode=BMM_Additive
    translationSpace=BCS_BoneSpace
    alphaBoolBlend {blendOption=Linear}
    boneToModify {boneName=calf_l_front}
    scale=<2 2 2>
    translation=<7 -5>
   }
   n74 = Get calf_r_front() -> (calf_r_front:double)
   n75 = "Transform (Modify) Bone - Bone: calf_r_front"(ComponentPose=@n73, Alpha=@n74, Translation="-7,-5,0", Rotation="0,0,0", Scale="2,2,2") {
    rotationMode=BMM_Additive
    rotationSpace=BCS_BoneSpace
    scaleMode=BMM_Additive
    scaleSpace=BCS_BoneSpace
    translationMode=BMM_Additive
    translationSpace=BCS_BoneSpace
    alphaBoolBlend {blendOption=Linear}
    boneToModify {boneName=calf_r_front}
    scale=<2 2 2>
    translation=<-7 -5>
   }
   n76 = "Transform (Modify) Bone - Bone: calf_l_back"(ComponentPose=@n75, Alpha=@n72, Translation="0,-12,0", Rotation="0,0,0", Scale="1,1,1") {
    rotationMode=BMM_Additive
    rotationSpace=BCS_BoneSpace
    scaleMode=BMM_Additive
    scaleSpace=BCS_BoneSpace
    translationMode=BMM_Additive
    translationSpace=BCS_BoneSpace
    alphaBoolBlend {blendOption=Linear}
    boneToModify {boneName=calf_l_back}
    translation {y=-12}
   }
   n77 = "Transform (Modify) Bone - Bone: calf_r_back"(ComponentPose=@n76, Alpha=@n74, Translation="0,12,0", Rotation="0,0,0", Scale="1,1,1") {
    rotationMode=BMM_Additive
    rotationSpace=BCS_BoneSpace
    scaleMode=BMM_Additive
    scaleSpace=BCS_BoneSpace
    translationMode=BMM_Additive
    translationSpace=BCS_BoneSpace
    alphaBoolBlend {blendOption=Linear}
    boneToModify {boneName=calf_r_back}
    translation {y=12}
   }
   n78 = "Constraint - flont_uppelalm_l (PP)"!(ComponentPose=@n77, Alpha=1, ConstraintWeights_0=0.5, ConstraintWeights_1:float=0.5) -> (Pose) {
    alphaBoolBlend {blendOption=Linear}
    boneToModify {boneName=flont_uppelalm_l}
    constraintSetup [{targetBone {boneName=upperarm_twist_01_l}}, {targetBone {boneName=clavicle_l}}]
    constraintWeights [1,1]
   }
   n79 = "Constraint - up_uppelalm_l (PP)"!(ComponentPose=@n78, Alpha=1, ConstraintWeights_0=0.5, ConstraintWeights_1:float=0.5) -> (Pose) {
    alphaBoolBlend {blendOption=Linear}
    boneToModify {boneName=up_uppelalm_l}
    constraintSetup [{targetBone {boneName=upperarm_twist_01_l}}, {targetBone {boneName=clavicle_l}}]
    constraintWeights [1,1]
   }
   n80 = "Constraint - front_upperarm_r (PP)"!(ComponentPose=@n79, Alpha=1, ConstraintWeights_0=0.3, ConstraintWeights_1:float=0.7) -> (Pose) {
    alphaBoolBlend {blendOption=Linear}
    boneToModify {boneName=front_upperarm_r}
    constraintSetup [{targetBone {boneName=upperarm_twist_01_r}}, {targetBone {boneName=clavicle_r}}]
    constraintWeights [1,1]
   }
   n81 = "Constraint - up_upperarm_r (PP)"!(ComponentPose=@n80, Alpha=1, ConstraintWeights_0=0.5, ConstraintWeights_1:float=0.5) -> (Pose) {
    alphaBoolBlend {blendOption=Linear}
    boneToModify {boneName=up_upperarm_r}
    constraintSetup [{targetBone {boneName=upperarm_twist_01_r}}, {targetBone {boneName=clavicle_r}}]
    constraintWeights [1,1]
   }
   n82 = Get Shoulder_rot_f_l() -> (Shoulder_rot_f_l:double)
   n83 = "Transform (Modify) Bone - Bone: flont_uppelalm_l"(ComponentPose=@n81, Alpha=@n82, Translation="-5,5,-5", Rotation="0,0,0", Scale="1,1,1") {
    rotationMode=BMM_Additive
    rotationSpace=BCS_BoneSpace
    scaleMode=BMM_Additive
    scaleSpace=BCS_BoneSpace
    translationMode=BMM_Additive
    translationSpace=BCS_BoneSpace
    alphaBoolBlend {blendOption=Linear}
    boneToModify {boneName=flont_uppelalm_l}
    translation=<-5 5 -5>
   }
   n84 = Get Shoulder_rot_f_r() -> (Shoulder_rot_f_r:double)
   n85 = "Transform (Modify) Bone - Bone: front_upperarm_r"(ComponentPose=@n83, Alpha=@n84, Translation="15,5,0", Rotation="0,0,0", Scale="1,1,1") {
    rotationMode=BMM_Additive
    rotationSpace=BCS_BoneSpace
    scaleMode=BMM_Additive
    scaleSpace=BCS_BoneSpace
    translationMode=BMM_Additive
    translationSpace=BCS_BoneSpace
    alphaBoolBlend {blendOption=Linear}
    boneToModify {boneName=front_upperarm_r}
    translation=<15 5>
   }
   n86 = "Transform (Modify) Bone - Bone: up_uppelalm_l"(ComponentPose=@n85, Alpha=Get Shoulder_rot_up_l() -> (Shoulder_rot_up_l:double), Translation="-15,0,2", Rotation="0,0,0", Scale="1,1,1") {
    rotationMode=BMM_Additive
    rotationSpace=BCS_BoneSpace
    scaleMode=BMM_Additive
    scaleSpace=BCS_BoneSpace
    translationMode=BMM_Additive
    translationSpace=BCS_BoneSpace
    alphaBoolBlend {blendOption=Linear}
    boneToModify {boneName=up_uppelalm_l}
    translation {x=-15, z=2}
   }
   n87 = "Transform (Modify) Bone - Bone: up_upperarm_r"(ComponentPose=@n86, Alpha=Get Shoulder_rot_up_r_0() -> (Shoulder_rot_up_r_0:double), Translation="-15,0,0", Rotation="0,0,0", Scale="1,1,1") {
    rotationMode=BMM_Additive
    rotationSpace=BCS_BoneSpace
    scaleMode=BMM_Additive
    scaleSpace=BCS_BoneSpace
    translationMode=BMM_Additive
    translationSpace=BCS_BoneSpace
    alphaBoolBlend {blendOption=Linear}
    boneToModify {boneName=up_upperarm_r}
    translation {x=-15}
   }
   n88 = "Transform (Modify) Bone - Bone: back_uppelalm_l"(ComponentPose=@n87, Alpha=@n82, Translation="0,1,0", Rotation="0,0,0", Scale="1.1,1.1,1.1") {
    rotationMode=BMM_Additive
    rotationSpace=BCS_BoneSpace
    scaleMode=BMM_Additive
    scaleSpace=BCS_BoneSpace
    translationMode=BMM_Additive
    translationSpace=BCS_BoneSpace
    alphaBoolBlend {blendOption=Linear}
    boneToModify {boneName=back_uppelalm_l}
    scale=<1.1 1.1 1.1>
    translation {y=1}
   }
   n89 = "Transform (Modify) Bone - Bone: back_upperarm_r"(ComponentPose=@n88, Alpha=@n84, Translation="0,6,0", Rotation="0,0,0", Scale="2,2,2") {
    rotationMode=BMM_Additive
    rotationSpace=BCS_BoneSpace
    scaleMode=BMM_Additive
    scaleSpace=BCS_BoneSpace
    translationMode=BMM_Additive
    translationSpace=BCS_BoneSpace
    alphaBoolBlend {blendOption=Linear}
    boneToModify {boneName=back_upperarm_r}
    scale=<2 2 2>
    translation {y=6}
   }
   n90 = "Constraint - upperarmDn_l (PP)"!(ComponentPose=@n89, Alpha=1, ConstraintWeights_0=0.7, ConstraintWeights_1:float=0.3) -> (Pose) {
    alphaBoolBlend {blendOption=Linear}
    boneToModify {boneName=upperarmDn_l}
    constraintSetup [{targetBone {boneName=spine_04}}, {targetBone {boneName=clavicle_l}}]
    constraintWeights [1,1]
   }
   n91 = "Constraint - upperarmDn_r (PP)"!(ComponentPose=@n90, Alpha=1, ConstraintWeights_0=0.7, ConstraintWeights_1:float=0.3) -> (Pose) {
    alphaBoolBlend {blendOption=Linear}
    boneToModify {boneName=upperarmDn_r}
    constraintSetup [{targetBone {boneName=spine_04}}, {targetBone {boneName=clavicle_r}}]
    constraintWeights [1,1]
   }
   n92 = "Transform (Modify) Bone - Bone: upperarmDn_l"(ComponentPose=@n91, Alpha=Get Shoulder_rot_dn_l() -> (Shoulder_rot_dn_l:double), Translation="-8,0,0", Rotation="0,0,0", Scale="1,1,1") {
    rotationMode=BMM_Additive
    rotationSpace=BCS_BoneSpace
    scaleMode=BMM_Additive
    scaleSpace=BCS_BoneSpace
    translationMode=BMM_Additive
    translationSpace=BCS_BoneSpace
    alphaBoolBlend {blendOption=Linear}
    boneToModify {boneName=upperarmDn_l}
    translation {x=-8}
   }
   n93 = "Transform (Modify) Bone - Bone: upperarmDn_r"(ComponentPose=@n92, Alpha=Get Shoulder_rot_dn_r() -> (Shoulder_rot_dn_r:double), Translation="8,0,0", Rotation="0,0,0", Scale="1,1,1") {
    rotationMode=BMM_Additive
    rotationSpace=BCS_BoneSpace
    scaleMode=BMM_Additive
    scaleSpace=BCS_BoneSpace
    translationMode=BMM_Additive
    translationSpace=BCS_BoneSpace
    alphaBoolBlend {blendOption=Linear}
    boneToModify {boneName=upperarmDn_r}
    translation {x=8}
   }
   n94 = "Transform (Modify) Bone - Bone: upperarmDn_l"(ComponentPose=@n93, Alpha=Get Shoulder_rot_up_l(), Translation="-3,0,15", Rotation="0,0,0", Scale="1,1,1") {
    rotationMode=BMM_Additive
    rotationSpace=BCS_BoneSpace
    scaleMode=BMM_Additive
    scaleSpace=BCS_BoneSpace
    translationMode=BMM_Additive
    translationSpace=BCS_BoneSpace
    alphaBoolBlend {blendOption=Linear}
    boneToModify {boneName=upperarmDn_l}
    translation {x=-3, z=15}
   }
   n95 = "Transform (Modify) Bone - Bone: upperarmDn_r"(ComponentPose=@n94, Alpha=Get Shoulder_rot_up_r() -> (Shoulder_rot_up_r:double), Translation="3,0,-15", Rotation="0,0,0", Scale="1,1,1") {
    rotationMode=BMM_Additive
    rotationSpace=BCS_BoneSpace
    scaleMode=BMM_Additive
    scaleSpace=BCS_BoneSpace
    translationMode=BMM_Additive
    translationSpace=BCS_BoneSpace
    alphaBoolBlend {blendOption=Linear}
    boneToModify {boneName=upperarmDn_r}
    translation {x=3, z=-15}
   }
   n96 = Get Shoulder_rot_back_l() -> (Shoulder_rot_back_l:double)
   n97 = "Transform (Modify) Bone - Bone: upperarmDn_l"(ComponentPose=@n95, Alpha=@n96, Translation="-5,-5,-2.5", Rotation="0,0,0", Scale="1,1,1") {
    rotationMode=BMM_Additive
    rotationSpace=BCS_BoneSpace
    scaleMode=BMM_Additive
    scaleSpace=BCS_BoneSpace
    translationMode=BMM_Additive
    translationSpace=BCS_BoneSpace
    alphaBoolBlend {blendOption=Linear}
    boneToModify {boneName=upperarmDn_l}
    translation=<-5 -5 -2.5>
   }
   n98 = Get Shoulder_rot_back_r() -> (Shoulder_rot_back_r:double)
   n99 = "Transform (Modify) Bone - Bone: upperarmDn_r"(ComponentPose=@n97, Alpha=@n98, Translation="5,5,2.5", Rotation="0,0,0", Scale="1,1,1") {
    rotationMode=BMM_Additive
    rotationSpace=BCS_BoneSpace
    scaleMode=BMM_Additive
    scaleSpace=BCS_BoneSpace
    translationMode=BMM_Additive
    translationSpace=BCS_BoneSpace
    alphaBoolBlend {blendOption=Linear}
    boneToModify {boneName=upperarmDn_r}
    translation=<5 5 2.5>
   }
   n100 = "Constraint - dyn_poncho_clav_roll_l_MSTR (PP)"!(ComponentPose=@n99, Alpha=1, ConstraintWeights_0=0.5, ConstraintWeights_1:float=0.5) -> (Pose) {
    alphaBoolBlend {blendOption=Linear}
    boneToModify {boneName=dyn_poncho_clav_roll_l_MSTR}
    constraintSetup [{targetBone {boneName=spine_05}}, {targetBone {boneName=clavicle_l}}]
    constraintWeights [1,1]
   }
   n101 = "Constraint - dyn_poncho_clav_roll_r_MSTR (PP)"!(ComponentPose=@n100, Alpha=1, ConstraintWeights_0=0.5, ConstraintWeights_1:float=0.5) -> (Pose) {
    alphaBoolBlend {blendOption=Linear}
    boneToModify {boneName=dyn_poncho_clav_roll_r_MSTR}
    constraintSetup [{targetBone {boneName=spine_05}}, {targetBone {boneName=clavicle_r}}]
    constraintWeights [1,1]
   }
   n102 = "Transform (Modify) Bone - Bone: dyn_poncho_clav_roll_l_MSTR"(ComponentPose=@n101, Alpha=@n96, Translation="0,0,0", Rotation="0,-5,10", Scale="1,1,1") {
    rotationMode=BMM_Additive
    rotationSpace=BCS_BoneSpace
    scaleMode=BMM_Additive
    scaleSpace=BCS_BoneSpace
    translationMode=BMM_Additive
    translationSpace=BCS_BoneSpace
    alphaBoolBlend {blendOption=Linear}
    boneToModify {boneName=dyn_poncho_clav_roll_l_MSTR}
    rotation {roll=10, yaw=-5}
   }
   n103 = "Transform (Modify) Bone - Bone: dyn_poncho_clav_roll_r_MSTR"(ComponentPose=@n102, Alpha=@n98, Translation="0,0,0", Rotation="0,5,-10", Scale="1,1,1") {
    rotationMode=BMM_Additive
    rotationSpace=BCS_BoneSpace
    scaleMode=BMM_Additive
    scaleSpace=BCS_BoneSpace
    translationMode=BMM_Additive
    translationSpace=BCS_BoneSpace
    alphaBoolBlend {blendOption=Linear}
    boneToModify {boneName=dyn_poncho_clav_roll_r_MSTR}
    rotation {roll=-10, yaw=5}
   }
   n104 = Get spine_04() -> (spine_04:double)
   n105 = "Transform (Modify) Bone - Bone: dyn_poncho_bk_l_02"(ComponentPose=@n103, Alpha=@n104, Translation="0,0,0", Rotation="15,0,0", Scale="1,1,1") {
    rotationMode=BMM_Additive
    rotationSpace=BCS_BoneSpace
    scaleMode=BMM_Additive
    scaleSpace=BCS_BoneSpace
    translationMode=BMM_Additive
    translationSpace=BCS_BoneSpace
    alphaBoolBlend {blendOption=Linear}
    boneToModify {boneName=dyn_poncho_bk_l_02}
    rotation {pitch=15}
   }
   n106 = "Transform (Modify) Bone - Bone: dyn_poncho_bk_r_02"(ComponentPose=@n105, Alpha=@n104, Translation="0,0,0", Rotation="15,0,0", Scale="1,1,1") {
    rotationMode=BMM_Additive
    rotationSpace=BCS_BoneSpace
    scaleMode=BMM_Additive
    scaleSpace=BCS_BoneSpace
    translationMode=BMM_Additive
    translationSpace=BCS_BoneSpace
    alphaBoolBlend {blendOption=Linear}
    boneToModify {boneName=dyn_poncho_bk_r_02}
    rotation {pitch=15}
   }
   n107 = "Transform (Modify) Bone - Bone: dyn_poncho_bk_02"(ComponentPose=@n106, Alpha=@n104, Translation="0,0,0", Rotation="20,0,0", Scale="1,1,1") {
    rotationMode=BMM_Additive
    rotationSpace=BCS_BoneSpace
    scaleMode=BMM_Additive
    scaleSpace=BCS_BoneSpace
    translationMode=BMM_Additive
    translationSpace=BCS_BoneSpace
    alphaBoolBlend {blendOption=Linear}
    boneToModify {boneName=dyn_poncho_bk_02}
    rotation {pitch=20}
   }
   n108 = Get spine_05() -> (spine_05:double)
   n109 = "Transform (Modify) Bone - Bone: dyn_poncho_bk_l_03"(ComponentPose=@n107, Alpha=@n108, Translation="0,0,0", Rotation="20,0,0", Scale="1,1,1") {
    rotationMode=BMM_Additive
    rotationSpace=BCS_BoneSpace
    scaleMode=BMM_Additive
    scaleSpace=BCS_BoneSpace
    translationMode=BMM_Additive
    translationSpace=BCS_BoneSpace
    alphaBoolBlend {blendOption=Linear}
    boneToModify {boneName=dyn_poncho_bk_l_03}
    rotation {pitch=20}
   }
   n110 = "Transform (Modify) Bone - Bone: dyn_poncho_bk_r_03"(ComponentPose=@n109, Alpha=@n108, Translation="0,0,0", Rotation="20,0,0", Scale="1,1,1") {
    rotationMode=BMM_Additive
    rotationSpace=BCS_BoneSpace
    scaleMode=BMM_Additive
    scaleSpace=BCS_BoneSpace
    translationMode=BMM_Additive
    translationSpace=BCS_BoneSpace
    alphaBoolBlend {blendOption=Linear}
    boneToModify {boneName=dyn_poncho_bk_r_03}
    rotation {pitch=20}
   }
   n111 = "Transform (Modify) Bone - Bone: dyn_poncho_bk_03"(ComponentPose=@n110, Alpha=@n108, Translation="0,0,0", Rotation="40,0,0", Scale="1,1,1") {
    rotationMode=BMM_Additive
    rotationSpace=BCS_BoneSpace
    scaleMode=BMM_Additive
    scaleSpace=BCS_BoneSpace
    translationMode=BMM_Additive
    translationSpace=BCS_BoneSpace
    alphaBoolBlend {blendOption=Linear}
    boneToModify {boneName=dyn_poncho_bk_03}
    rotation {pitch=40}
   }
   n112 = Anim Dynamics - Bone: dyn_hat_bk_01!(ComponentPose=@n111, bAlphaBoolEnabled:bool=True, ExternalForce="0,0,0") -> (Pose) {
    alphaInputType=Bool
    angularDampingOverride=0.999
    angularSpringConstant=800
    bAngularSpring=true
    bGravityOverrideInSimSpace=true
    bOverrideAngularDamping=true
    bUseGravityOverride=true
    gravityScale=0.2
    simulationSpace=RootRelative
    alphaBoolBlend {blendOption=Linear}
    boundBone {boneName=dyn_hat_bk_01}
    componentAppliedLinearAccClamp=<6000 6000 6000>
    componentLinearAccScale=<0.8 0.8 0.8>
    componentLinearVelScale=<0.8 0.8 0.8>
    gravityOverride {z=-980}
    physicsBodyDefinitions [{
      sphereCollisionRadius=0
      boundBone {boneName=dyn_hat_bk_01}
      boxExtents=<23 23 23>
      constraintSetup {angularTargetAxis=AxisY, angularLimitsMax=<5 25 5>, angularLimitsMin=<-5 -90 -5>, angularTarget=<0 0 0>}
      localJointOffset=<5 0.5 -0.2>
     }]
    relativeSpaceBone {boneName=spine_05}
   }
   n113 = Anim Dynamics - Bone: dyn_hat_bk_02!(ComponentPose=@n112, bAlphaBoolEnabled:bool=True, ExternalForce="0,0,0") -> (Pose) {
    alphaInputType=Bool
    angularDampingOverride=0.999
    angularSpringConstant=800
    bAngularSpring=true
    bGravityOverrideInSimSpace=true
    bOverrideAngularDamping=true
    bUseGravityOverride=true
    gravityScale=0.2
    simulationSpace=RootRelative
    alphaBoolBlend {blendOption=Linear}
    boundBone {boneName=dyn_hat_bk_02}
    componentAppliedLinearAccClamp=<6000 6000 6000>
    componentLinearAccScale=<0.8 0.8 0.8>
    componentLinearVelScale=<0.8 0.8 0.8>
    gravityOverride {z=-980}
    physicsBodyDefinitions [{
      sphereCollisionRadius=0
      boundBone {boneName=dyn_hat_bk_02}
      boxExtents=<23 23 23>
      constraintSetup {angularTargetAxis=AxisY, angularLimitsMax=<5 25 5>, angularLimitsMin=<-5 -90 -5>, angularTarget=<0 0 0>}
      localJointOffset=<5 0.5 -0.2>
     }]
    relativeSpaceBone {boneName=spine_05}
   }
   n114 = "Constraint - lowerarm_l_front (RR)"!(ComponentPose=@n113, Alpha=1, ConstraintWeights_0=0.5, ConstraintWeights_1:float=0.5) -> (Pose) {
    alphaBoolBlend {blendOption=Linear}
    boneToModify {boneName=lowerarm_l_front}
    constraintSetup [{transformType=Rotation, targetBone {boneName=lowerarm_twist_02_l}}, {transformType=Rotation, targetBone {boneName=upperarm_twist_02_l}}]
    constraintWeights [1,1]
   }
   n115 = "Constraint - lowerarm_r_front (RR)"!(ComponentPose=@n114, Alpha=1, ConstraintWeights_0=0.5, ConstraintWeights_1:float=0.5) -> (Pose) {
    alphaBoolBlend {blendOption=Linear}
    boneToModify {boneName=lowerarm_r_front}
    constraintSetup [{transformType=Rotation, targetBone {boneName=lowerarm_twist_02_r}}, {transformType=Rotation, targetBone {boneName=upperarm_twist_02_r}}]
    constraintWeights [1,1]
   }
   n116 = "Constraint - lowerarm_l_back (RR)"!(ComponentPose=@n115, Alpha=1, ConstraintWeights_0=0.5, ConstraintWeights_1:float=0.5) -> (Pose) {
    alphaBoolBlend {blendOption=Linear}
    boneToModify {boneName=lowerarm_l_back}
    constraintSetup [{transformType=Rotation, targetBone {boneName=lowerarm_twist_02_l}}, {transformType=Rotation, targetBone {boneName=upperarm_twist_02_l}}]
    constraintWeights [1,1]
   }
   n117 = "Constraint - lowerarm_r_back (RR)"!(ComponentPose=@n116, Alpha=1, ConstraintWeights_0=0.5, ConstraintWeights_1:float=0.5) -> (Pose) {
    alphaBoolBlend {blendOption=Linear}
    boneToModify {boneName=lowerarm_r_back}
    constraintSetup [{transformType=Rotation, targetBone {boneName=lowerarm_twist_02_r}}, {transformType=Rotation, targetBone {boneName=upperarm_twist_02_r}}]
    constraintWeights [1,1]
   }
   n118 = Get lowerarm_l_front() -> (lowerarm_l_front:double)
   n119 = "Transform (Modify) Bone - Bone: lowerarm_l_front"(ComponentPose=@n117, Alpha=@n118, Translation="7,7,0", Rotation="0,0,0", Scale="1.2,1.2,1.2") {
    rotationMode=BMM_Additive
    rotationSpace=BCS_BoneSpace
    scaleMode=BMM_Additive
    scaleSpace=BCS_BoneSpace
    translationMode=BMM_Additive
    translationSpace=BCS_BoneSpace
    alphaBoolBlend {blendOption=Linear}
    boneToModify {boneName=lowerarm_l_front}
    scale=<1.2 1.2 1.2>
    translation=<7 7>
   }
   n120 = Get lowerarm_r_front() -> (lowerarm_r_front:double)
   n121 = "Transform (Modify) Bone - Bone: lowerarm_r_front"(ComponentPose=@n119, Alpha=@n120, Translation="-7,-7,0", Rotation="0,0,0", Scale="1,1,1") {
    rotationMode=BMM_Additive
    rotationSpace=BCS_BoneSpace
    scaleMode=BMM_Additive
    scaleSpace=BCS_BoneSpace
    translationMode=BMM_Additive
    translationSpace=BCS_BoneSpace
    alphaBoolBlend {blendOption=Linear}
    boneToModify {boneName=lowerarm_r_front}
    translation=<-7 -7>
   }
   n122 = "Transform (Modify) Bone - Bone: lowerarm_l_back"(ComponentPose=@n121, Alpha=@n118, Translation="0,0,0", Rotation="0,0,0", Scale="1,1,1") {
    rotationMode=BMM_Additive
    rotationSpace=BCS_BoneSpace
    scaleMode=BMM_Additive
    scaleSpace=BCS_BoneSpace
    translationMode=BMM_Additive
    translationSpace=BCS_BoneSpace
    alphaBoolBlend {blendOption=Linear}
    boneToModify {boneName=lowerarm_l_back}
   }
   n123 = "Transform (Modify) Bone - Bone: lowerarm_r_back"(ComponentPose=@n122, Alpha=@n120, Translation="0,0,0", Rotation="0,0,0", Scale="1,1,1") {
    rotationMode=BMM_Additive
    rotationSpace=BCS_BoneSpace
    scaleMode=BMM_Additive
    scaleSpace=BCS_BoneSpace
    translationMode=BMM_Additive
    translationSpace=BCS_BoneSpace
    alphaBoolBlend {blendOption=Linear}
    boneToModify {boneName=lowerarm_r_back}
   }
   n124 = Control Rig(Source=Component To Local(ComponentPose=@n123), Alpha=Get idle() -> (idle:double)) {
    alphaBoolBlend {blendOption=Linear}
    controlRigAssetReference {blueprintRigClass=&/M_MED_NightHawk_CtrlRig.M_MED_NightHawk_CtrlRig_C}
   }
   Save cached pose 'CrouchingCache'(Pose=@n124) {cacheName=CrouchingCache}
  ]}
 EventGraph {nodes [
   n1 = Event OnBeginSkydiving()
   Set CapeHide(execute=[@n1.then, @n3.else, @n21.else], CapeHide:double=0) -> (Output_Get:double)
   Event OnEndSkydiving()
   n2 = CharacterItemDefinitionHasAnyMetaTag(Tags=Make Array("[0]"=(TagName="Cosmetics.MetaTags.NightHawk")))
   > n3 = Branch(Condition=@n2)
   Set CapeHide(execute=[@n3.then, @n21.then], CapeHide=1)
   Event BlueprintInitializeAnimation()
   > Set Skel_Mesh(Skel_Mesh=GetOwnerSkeletalMeshComponent())
   > n4 = Cast To PlayerPawn_Athena(Object=TryGetPawnOwner())
   > n5 = Set Pawn Owner(Pawn Owner:Fort Player Pawn Athena Object Reference=@n4) -> (Output_Get:Fort Player Pawn Athena Object Reference)
   > Set Shoulder_Mesh(Shoulder_Mesh:Skeletal Mesh Component Object Reference=GetSkeletalMeshForPartType(self=@n5)) -> (Output_Get:Skeletal Mesh Component Object Reference)
   > Set Head_Mesh(Head_Mesh:Skeletal Mesh Component Object Reference=GetSkeletalMeshForPartType(self=@n5, PartType=Head)) -> (Output_Get:Skeletal Mesh Component Object Reference)
   Event BlueprintUpdateAnimation()
   > n6 = Sequence!(execute) -> (then_0, then_1, then_2:Exec, then_3:Exec, then_4:Exec, then_5:Exec, then_6:Exec)
   n7 = Get Skel_Mesh()
   > Set thigh_shape_fn_l(thigh_shape_fn_l:double=GetAlphaFromReferenceTransformDelta(SkeletalMeshComponent=@n7, BoneName=thigh_l, BaseName=pelvis, InputRange=(X=0,Y=-140), TwistAxis="0,0,1")) -> (Output_Get:double)
   > Set thigh_shape_fn_r(thigh_shape_fn_r:double=GetAlphaFromReferenceTransformDelta(SkeletalMeshComponent=@n7, BoneName=thigh_r, BaseName=pelvis, InputRange=(X=0,Y=-140), TwistAxis="0,0,1")) -> (Output_Get:double)
   n8 = Get Skel_Mesh()
   > Set thigh_shape_bk_l(thigh_shape_bk_l:double=GetAlphaFromReferenceTransformDelta(SkeletalMeshComponent=@n8, BoneName=thigh_l, BaseName=pelvis, InputRange=(X=0,Y=-110), TwistAxis="0,0,1")) -> (Output_Get:double)
   > Set thigh_shape_bk_r(thigh_shape_bk_r:double=GetAlphaFromReferenceTransformDelta(SkeletalMeshComponent=@n8, BoneName=thigh_r, BaseName=pelvis, InputRange=(X=0,Y=-110), TwistAxis="0,0,1")) -> (Output_Get:double)
   n9 = Get Skel_Mesh()
   > Set thigh_shape_lt_l(thigh_shape_lt_l:double=GetAlphaFromReferenceTransformDelta(SkeletalMeshComponent=@n9, BoneName=thigh_l, BaseName=pelvis, InputRange=(X=-90,Y=0), OutputRange=(X=1,Y=0), TwistAxis="0,0,1")) -> (Output_Get:double)
   > Set thigh_shape_rt_r(thigh_shape_rt_r:double=GetAlphaFromReferenceTransformDelta(SkeletalMeshComponent=@n9, BoneName=thigh_r, BaseName=pelvis, InputRange=(X=-90,Y=0), OutputRange=(X=1,Y=0), TwistAxis="0,0,1")) -> (Output_Get:double)
   n10 = Get Skel_Mesh()
   > Set thigh_shape_bk_l_0(thigh_shape_bk_l_0:double=GetAlphaFromReferenceTransformDelta(SkeletalMeshComponent=@n10, BoneName=thigh_l, BaseName=pelvis, InputRange=(X=0,Y=60), TwistAxis="0,0,1")) -> (Output_Get:double)
   > Set thigh_shape_bk_r_0(thigh_shape_bk_r_0:double=GetAlphaFromReferenceTransformDelta(SkeletalMeshComponent=@n10, BoneName=thigh_r, BaseName=pelvis, InputRange=(X=0,Y=60), TwistAxis="0,0,1")) -> (Output_Get:double)
   n11 = Get Skel_Mesh()
   > Set thigh_shape_fn_l_0(thigh_shape_fn_l_0:double=GetAlphaFromReferenceTransformDelta(SkeletalMeshComponent=@n11, BoneName=thigh_twist_01_l, BaseName=pelvis, InputRange=(X=10,Y=-160), TwistAxis="0,0,1")) -> (Output_Get:double)
   > Set thigh_shape_fn_r_0(thigh_shape_fn_r_0:double=GetAlphaFromReferenceTransformDelta(SkeletalMeshComponent=@n11, BoneName=thigh_twist_01_r, BaseName=pelvis, InputRange=(X=10,Y=-160), TwistAxis="0,0,1")) -> (Output_Get:double)
   n12 = Get Shoulder_Mesh() -> (Shoulder_Mesh:Skeletal Mesh Component Object Reference)
   > Set Shoulder_rot_f_r(Shoulder_rot_f_r:double=GetAlphaFromReferenceTransformDelta(SkeletalMeshComponent=@n12, BoneName=upperarm_r, BaseName=clavicle_r, InputRange=(X=0,Y=110), TwistAxis="0,0,1")) -> (Output_Get:double)
   > Set Shoulder_rot_f_l(Shoulder_rot_f_l:double=GetAlphaFromReferenceTransformDelta(SkeletalMeshComponent=@n12, BoneName=upperarm_l, BaseName=clavicle_l, InputRange=(X=0,Y=110), TwistAxis="0,0,1")) -> (Output_Get:double)
   n13 = Get Shoulder_Mesh()
   > Set Shoulder_rot_up_r(Shoulder_rot_up_r:double=GetAlphaFromReferenceTransformDelta(
     SkeletalMeshComponent=@n13
     BoneName=upperarm_r
     BaseName=spine_05
     DeltaComponentType=EulerY
     InputRange=(X=60,Y=-60)
     OutputRange=(X=1,Y=0)
     TwistAxis="0,0,1"
    )) -> (Output_Get:double)
   > Set Shoulder_rot_up_l(Shoulder_rot_up_l:double=GetAlphaFromReferenceTransformDelta(
     SkeletalMeshComponent=@n13
     BoneName=upperarm_l
     BaseName=spine_05
     DeltaComponentType=EulerY
     InputRange=(X=60,Y=-60)
     OutputRange=(X=1,Y=0)
     TwistAxis="0,0,1"
    )) -> (Output_Get:double)
   n14 = Get Shoulder_Mesh()
   > Set Shoulder_rot_dn_r(Shoulder_rot_dn_r:double=GetAlphaFromReferenceTransformDelta(SkeletalMeshComponent=@n14, BoneName=upperarm_r, BaseName=clavicle_r, DeltaComponentType=EulerX, InputRange=(X=-15,Y=-90), TwistAxis="0,0,1")) -> (Output_Get:double)
   > Set Shoulder_rot_up_r_0(Shoulder_rot_up_r_0:double=GetAlphaFromReferenceTransformDelta(SkeletalMeshComponent=Get Shoulder_Mesh(), BoneName=upperarm_r, BaseName=clavicle_r, InputRange=(X=0,Y=150), TwistAxis="0,0,1")) -> (Output_Get:double)
   > Set Shoulder_rot_up_r_1(Shoulder_rot_up_r_1:double=GetAlphaFromReferenceTransformDelta(
     SkeletalMeshComponent=Get Shoulder_Mesh()
     BoneName=upperarm_r
     BaseName=clavicle_r
     DeltaComponentType=EulerY
     InputRange=(X=0,Y=30)
     TwistAxis="0,0,1"
    )) -> (Output_Get:double)
   n15 = Get Shoulder_Mesh()
   > Set Shoulder_rot_back_r(Shoulder_rot_back_r:double=GetAlphaFromReferenceTransformDelta(SkeletalMeshComponent=@n15, BoneName=upperarm_twist_01_r, BaseName=clavicle_r, InputRange=(X=10,Y=-30), TwistAxis="0,0,1")) -> (Output_Get:double)
   > Set Shoulder_rot_back_l(Shoulder_rot_back_l:double=GetAlphaFromReferenceTransformDelta(SkeletalMeshComponent=@n15, BoneName=upperarm_twist_01_l, BaseName=clavicle_l, InputRange=(X=10,Y=-30), TwistAxis="0,0,1")) -> (Output_Get:double)
   > Set Shoulder_rot_dn_l(Shoulder_rot_dn_l:double=GetAlphaFromReferenceTransformDelta(SkeletalMeshComponent=@n14, BoneName=upperarm_l, BaseName=clavicle_l, DeltaComponentType=EulerX, InputRange=(X=-15,Y=-90), TwistAxis="0,0,1")) -> (Output_Get:double)
   > n16 = Sequence!(execute) -> (then_0, then_1, then_2, then_3)
   Set lowerarm_r_front(
    execute=@n16.then_1
    lowerarm_r_front:double=GetAlphaFromReferenceTransformDelta(SkeletalMeshComponent=Get Skel_Mesh(), BoneName=upperarm_r, BaseName=lowerarm_r, InputRange=(X=0,Y=-90), TwistAxis="0,0,1")
   ) -> (Output_Get:double)
   > Set calf_r_front(calf_r_front:double=GetAlphaFromReferenceTransformDelta(SkeletalMeshComponent=Get Skel_Mesh(), BoneName=thigh_r, BaseName=calf_r, InputRange=(X=0,Y=-90), TwistAxis="0,0,1")) -> (Output_Get:double)
   Set lowerarm_l_front(
    execute=@n16.then_0
    lowerarm_l_front:double=GetAlphaFromReferenceTransformDelta(SkeletalMeshComponent=Get Skel_Mesh(), BoneName=upperarm_l, BaseName=lowerarm_l, InputRange=(X=0,Y=-45), TwistAxis="0,0,1")
   ) -> (Output_Get:double)
   > Set calf_l_front(calf_l_front:double=GetAlphaFromReferenceTransformDelta(SkeletalMeshComponent=Get Skel_Mesh(), BoneName=thigh_l, BaseName=calf_l, InputRange=(X=0,Y=-90), TwistAxis="0,0,1")) -> (Output_Get:double)
   > Set neck_rot_up(neck_rot_up:double=GetAlphaFromReferenceTransformDelta(
     SkeletalMeshComponent=Get Head_Mesh() -> (Head_Mesh:Skeletal Mesh Component Object Reference)
     BoneName=head
     BaseName=spine_05
     InputRange=(X=-20,Y=-80)
     TwistAxis="0,0,1"
    )) -> (Output_Get:double)
   > Set neck_rot_r(neck_rot_r:double=GetAlphaFromReferenceTransformDelta(
     SkeletalMeshComponent=Get Head_Mesh()
     BoneName=head
     BaseName=spine_05
     DeltaComponentType=EulerY
     InputRange=(X=20,Y=60)
     TwistAxis="0,0,1"
    )) -> (Output_Get:double)
   > Set neck_rot_l(neck_rot_l:double=GetAlphaFromReferenceTransformDelta(SkeletalMeshComponent=Get Head_Mesh(), BoneName=head, BaseName=spine_05, DeltaComponentType=EulerX, InputRange=(X=0,Y=60), TwistAxis="0,0,1")) -> (Output_Get:double)
   n17 = Gravity Override State Function DEPRECATED(
    execute=@n6.then_1
    CharacterPartAnimInstance=Self-Reference()
    GameTimeInSeconds=GetGameTimeInSeconds()
    GravityOverrideParams=Get Gravity Override Params()
   )
   > Set GravityOverride(GravityOverride=@n17)
   n18 = NOT Boolean(A=Get bIsSkyDiving())
   n19 = Get bIsMoving2D()
   n20 = Branch(execute=@n6.then_2, Condition=OR Boolean!(A=Get bIsParachuteOpened(), B=AND Boolean(A=@n19, B=@n18), C:bool) -> (ReturnValue))
   > Set spine_04(spine_04:double=0.6) -> (Output_Get:double)
   Set spine_04(
    execute=@n20.else
    spine_04=GetAlphaFromReferenceTransformDelta(SkeletalMeshComponent=Get Skel_Mesh(), BoneName=spine_04, BaseName=pelvis, InputRange=(X=0,Y=-30), TwistAxis="0,0,1")
   )
   > Set spine_05(spine_05:double=GetAlphaFromReferenceTransformDelta(SkeletalMeshComponent=Get Skel_Mesh(), BoneName=spine_05, BaseName=pelvis, InputRange=(X=0,Y=-30), TwistAxis="0,0,1")) -> (Output_Get:double)
   Get Dynamic Material Instance(execute=@n6.then_3, SkeletalMeshComponent=GetPartSkeletalMeshComponent(), ElementIndex=1)
   > Set CapeHide_Tag(CapeHide_Tag:bool=@n2) -> (Output_Get:bool)
   > n21 = Branch(Condition=@n2)
   n22 = Branch(execute=@n6.then_4, Condition=Get bIsPlayingEmote())
   > Set idle(idle:double=0) -> (Output_Get:double)
   n23 = Branch(execute=@n22.else, Condition=Get bIsInFrontEnd())
   > Set idle(idle=1)
   Set idle(execute=@n23.else, idle=0)
   n24 = Get Skel_Mesh()
   Set ClearRadiusEmotePlaying(
    execute=@n6.then_5
    ClearRadiusEmotePlaying:bool=Montage_IsPlaying(self=GetAnimInstance(self=@n24), Montage=/BRCosmetics/Animation/Game/MainPlayer/Emotes/ClearRadius/Emote_ClearRadius_CMM_Cape_M)
   ) -> (Output_Get:bool)
   n25 = Set Not_isWearingPackBool(execute=@n6.then_6, Not_isWearingPackBool:bool=NOT Boolean(A=Get isWearingPackBool())) -> (Output_Get:bool)
   n26 = Get isWearingPackBool()
   n27 = AND Boolean(A=Get bIsSkyDiving(), B=NOT Boolean(A=Get isWearingPackBool()))
   n28 = Get bIsParachuteOpened()
   > Set isWearingPackBool_1(isWearingPackBool_1:double=SelectFloat(A=SelectFloat(A=0.1, B=0.5, bPickA=@n26), B="To Float (Boolean)"(InBool=@n27), bPickA=@n28)) -> (Output_Get:double)
   > Set bIsParachuteOpened_0(bIsParachuteOpened_0:double="To Float (Boolean)"(InBool=@n28)) -> (Output_Get:double)
   n29 = NOT Boolean(A=Get isWearingPackBool())
   n30 = Get bIsCrouching()
   > Set Not Wearing And IsCrouch(Not Wearing And IsCrouch:double="To Float (Boolean)"(InBool=AND Boolean(A=@n30, B=@n29))) -> (Output_Get:double)
   n31 = OR Boolean!(A=Get bIsCrouching(), B=Get bIsSkyDiving(), C=Get isWearingPackBool(), D:bool) -> (ReturnValue)
   > Set IsWearing Or IsCrouch Or IsSkyDiving(IsWearing Or IsCrouch Or IsSkyDiving:double="To Float (Boolean)"(InBool=NOT Boolean(A=@n31))) -> (Output_Get:double)
   > n32 = Branch(Condition=@n25)
   > Set isWearingPackBool_0(isWearingPackBool_0:double=0.5) -> (Output_Get:double)
   Set isWearingPackBool_0(execute=@n32.else, isWearingPackBool_0=0)
   Event OnCharacterCustomizationLoaded()
   n33 = CheckForBackbling()
   > Set isWearingPackBool(isWearingPackBool:bool=OR Boolean!(A=@n33, B=@n33.bIsBackblingACape, C=@n33.bIsBackblingLong) -> (ReturnValue)) -> (Output_Get:bool)
   > Set is_WearingCape(is_WearingCape:bool=@n33.bIsBackblingLong) -> (Output_Get:bool)
   > Set FrontEndCape(FrontEndCape:bool=AND Boolean(A=@n33.bIsBackblingLong, B=Get bIsInFrontEnd())) -> (Output_Get:bool)
   AND Boolean(A=@n33, B=NOT Boolean(A=@n33.bIsBackblingLong))
  ]}
]
variables [
 bIsParachuteOpened_0 {type=number}
 calf_l_front {type=number}
 calf_r_front {type=number}
 CapeHide {type=number}
 CapeHide_Tag {type=boolean}
 ClearRadiusEmotePlaying {type=boolean}
 FrontEndCape {type=boolean}
 Gravity Override Params {
  type=GravityOverrideParamsStruct
  default {
   skydive_DivingDown {gravityOverride=<0 0 20>}
   skydive_FloatingDown {gravityOverride=<0 0 400>, windAmplitude=<200 -500 200>, windOffset=<0 -600 200>}
   skydive_GoingUp {gravityOverride=<0 0 50>, windAmplitude=<0 0 0>, windFrequency=<0 0 0>}
   skydive_Parachuting {gravityOverride=<0 0 20>}
   skydive_Surfing {gravityOverride=<500 -300 20>}
  }
 }
 GravityOverride {type=Vector}
 Head_Mesh {type=SkeletalMeshComponent}
 idle {type=number}
 is_WearingCape {type=boolean}
 IsWearing Or IsCrouch Or IsSkyDiving {type=number}
 isWearingPackBool {type=boolean}
 isWearingPackBool_0 {type=number}
 isWearingPackBool_1 {type=number}
 lowerarm_l_back {type=number}
 lowerarm_l_front {type=number}
 lowerarm_r_back {type=number}
 lowerarm_r_front {type=number}
 neck_rot_l {type=number}
 neck_rot_r {type=number}
 neck_rot_up {type=number}
 Not Wearing And IsCrouch {type=number}
 Not_isWearingPackBool {type=boolean}
 offset_weapon {type=number}
 Pawn Owner {type=FortPlayerPawnAthena}
 Shoulder_Mesh {type=SkeletalMeshComponent}
 Shoulder_rot_back_l {type=number}
 Shoulder_rot_back_r {type=number}
 Shoulder_rot_dn_l {type=number}
 Shoulder_rot_dn_r {type=number}
 Shoulder_rot_f_l {type=number}
 Shoulder_rot_f_l_0 {type=number}
 Shoulder_rot_f_r {type=number}
 Shoulder_rot_f_r_0 {type=number}
 Shoulder_rot_up_l {type=number}
 Shoulder_rot_up_r {type=number}
 Shoulder_rot_up_r_0 {type=number}
 Shoulder_rot_up_r_1 {type=number}
 Skel_Mesh {type=SkeletalMeshComponent}
 spine_04 {type=number}
 spine_04_0 {type=number}
 spine_05 {type=number}
 thigh_shape_bk_l {type=number}
 thigh_shape_bk_l_0 {type=number}
 thigh_shape_bk_r {type=number}
 thigh_shape_bk_r_0 {type=number}
 thigh_shape_fn_l {type=number}
 thigh_shape_fn_l_0 {type=number}
 thigh_shape_fn_r {type=number}
 thigh_shape_fn_r_0 {type=number}
 thigh_shape_lt_l {type=number}
 thigh_shape_rt_r {type=number}
]
