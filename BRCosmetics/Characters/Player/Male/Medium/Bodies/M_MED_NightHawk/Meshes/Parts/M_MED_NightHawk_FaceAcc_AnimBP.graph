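# Animation Blueprint for the NightHawk medium male face accessory mesh that adds secondary physics motion to its hair bones.
# It copies the pose from the attached parent mesh, then runs eight chained single-bone Anim Dynamics nodes over the left and right HairA and HairB bones.
# Each node simulates root-relative with angular springs, high damping, reduced gravity and angular limits, blended at half alpha.
# The event graph has no logic.
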AnimBlueprint /BRCosmetics/Characters/Player/Male/Medium/Bodies/M_MED_NightHawk/Meshes/Parts/M_MED_NightHawk_FaceAcc_AnimBP
target_skeleton=&/M_MED_NightHawk_FaceAcc_Skeleton
graphs [
 AnimGraph {nodes [
   n1 = Anim Dynamics - Bone: dyn_HairA01_l(ComponentPose=Local To Component(LocalPose=Copy Pose From Mesh() {bUseAttachedParent=true}), Alpha=0.5, ExternalForce="0,0,0") {
    angularDampingOverride=0.999
    angularSpringConstant=800
    bAngularSpring=true
    bGravityOverrideInSimSpace=true
    bOverrideAngularDamping=true
    bUseGravityOverride=true
    gravityScale=0.2
    simulationSpace=RootRelative
    alphaBoolBlend {blendOption=Linear}
    boundBone {boneName=dyn_HairA01_l}
    componentAppliedLinearAccClamp=<4000 4000 4000>
    componentLinearAccScale=<0.8 0.8 0.8>
    componentLinearVelScale=<0.8 0.8 0.8>
    gravityOverride {z=-980}
    physicsBodyDefinitions [{
      sphereCollisionRadius=0
      boundBone {boneName=dyn_HairA01_l}
      boxExtents=<20 20 20>
      constraintSetup {angularTargetAxis=AxisY, angularLimitsMax=<5 10 45>, angularLimitsMin=<-5 -10 -7>, angularTarget=<0 0 0>}
      localJointOffset=<5 -1 -0.2>
     }]
    relativeSpaceBone {boneName=spine_05}
   }
   n2 = Anim Dynamics - Bone: dyn_HairA02_l(ComponentPose=@n1, Alpha=0.5, ExternalForce="0,0,0") {
    angularDampingOverride=0.999
    angularSpringConstant=800
    bAngularSpring=true
    bGravityOverrideInSimSpace=true
    bOverrideAngularDamping=true
    bUseGravityOverride=true
    gravityScale=0.2
    simulationSpace=RootRelative
    alphaBoolBlend {blendOption=Linear}
    boundBone {boneName=dyn_HairA02_l}
    componentAppliedLinearAccClamp=<4000 4000 4000>
    componentLinearAccScale=<0.8 0.8 0.8>
    componentLinearVelScale=<0.8 0.8 0.8>
    gravityOverride {z=-980}
    physicsBodyDefinitions [{
      sphereCollisionRadius=0
      boundBone {boneName=dyn_HairA02_l}
      boxExtents=<20 20 20>
      constraintSetup {angularTargetAxis=AxisY, angularLimitsMax=<5 10 45>, angularLimitsMin=<-5 -10 -7>, angularTarget=<0 0 0>}
      localJointOffset=<5 -1 -0.2>
     }]
    relativeSpaceBone {boneName=spine_05}
   }
   n3 = Anim Dynamics - Bone: dyn_HairA01_r(ComponentPose=@n2, Alpha=0.5, ExternalForce="0,0,0") {
    angularDampingOverride=0.999
    angularSpringConstant=800
    bAngularSpring=true
    bGravityOverrideInSimSpace=true
    bOverrideAngularDamping=true
    bUseGravityOverride=true
    gravityScale=0.2
    simulationSpace=RootRelative
    alphaBoolBlend {blendOption=Linear}
    boundBone {boneName=dyn_HairA01_r}
    componentAppliedLinearAccClamp=<4000 4000 4000>
    componentLinearAccScale=<0.8 0.8 0.8>
    componentLinearVelScale=<0.8 0.8 0.8>
    gravityOverride {z=-980}
    physicsBodyDefinitions [{
      sphereCollisionRadius=0
      boundBone {boneName=dyn_HairA01_r}
      boxExtents=<15 15 15>
      constraintSetup {angularTargetAxis=AxisY, angularLimitsMax=<5 10 45>, angularLimitsMin=<-5 -10 -7>, angularTarget=<0 0 0>}
      localJointOffset=<5 -0.5 0.5>
     }]
    relativeSpaceBone {boneName=spine_05}
   }
   n4 = Anim Dynamics - Bone: dyn_HairA02_r(ComponentPose=@n3, Alpha=0.5, ExternalForce="0,0,0") {
    angularDampingOverride=0.999
    angularSpringConstant=800
    bAngularSpring=true
    bGravityOverrideInSimSpace=true
    bOverrideAngularDamping=true
    bUseGravityOverride=true
    gravityScale=0.2
    simulationSpace=RootRelative
    alphaBoolBlend {blendOption=Linear}
    boundBone {boneName=dyn_HairA02_r}
    componentAppliedLinearAccClamp=<4000 4000 4000>
    componentLinearAccScale=<0.8 0.8 0.8>
    componentLinearVelScale=<0.8 0.8 0.8>
    gravityOverride {z=-980}
    physicsBodyDefinitions [{
      sphereCollisionRadius=0
      boundBone {boneName=dyn_HairA02_r}
      boxExtents=<15 15 15>
      constraintSetup {angularTargetAxis=AxisY, angularLimitsMax=<5 10 45>, angularLimitsMin=<-5 -10 -7>, angularTarget=<0 0 0>}
      localJointOffset=<5 -0.5 0.5>
     }]
    relativeSpaceBone {boneName=spine_05}
   }
   n5 = Anim Dynamics - Bone: dyn_HairB01_l(ComponentPose=@n4, Alpha=0.5, ExternalForce="0,0,0") {
    angularDampingOverride=0.999
    angularSpringConstant=800
    bAngularSpring=true
    bGravityOverrideInSimSpace=true
    bOverrideAngularDamping=true
    bUseGravityOverride=true
    gravityScale=0.2
    simulationSpace=RootRelative
    alphaBoolBlend {blendOption=Linear}
    boundBone {boneName=dyn_HairB01_l}
    componentAppliedLinearAccClamp=<3000 3000 3000>
    componentLinearAccScale=<0.8 0.8 0.8>
    componentLinearVelScale=<0.8 0.8 0.8>
    gravityOverride {z=-980}
    physicsBodyDefinitions [{
      sphereCollisionRadius=0
      boundBone {boneName=dyn_HairB01_l}
      boxExtents=<25 25 25>
      constraintSetup {angularTargetAxis=AxisY, angularLimitsMax=<5 10 45>, angularLimitsMin=<-5 -10 -7>, angularTarget=<0 0 0>}
      localJointOffset=<5 -1 -0.2>
     }]
    relativeSpaceBone {boneName=spine_05}
   }
   n6 = Anim Dynamics - Bone: dyn_HairB02_l(ComponentPose=@n5, Alpha=0.5, ExternalForce="0,0,0") {
    angularDampingOverride=0.999
    angularSpringConstant=800
    bAngularSpring=true
    bGravityOverrideInSimSpace=true
    bOverrideAngularDamping=true
    bUseGravityOverride=true
    gravityScale=0.2
    simulationSpace=RootRelative
    alphaBoolBlend {blendOption=Linear}
    boundBone {boneName=dyn_HairB02_l}
    componentAppliedLinearAccClamp=<3000 3000 3000>
    componentLinearAccScale=<0.8 0.8 0.8>
    componentLinearVelScale=<0.8 0.8 0.8>
    gravityOverride {z=-980}
    physicsBodyDefinitions [{
      sphereCollisionRadius=0
      boundBone {boneName=dyn_HairB02_l}
      boxExtents=<25 25 25>
      constraintSetup {angularTargetAxis=AxisY, angularLimitsMax=<5 10 45>, angularLimitsMin=<-5 -10 -7>, angularTarget=<0 0 0>}
      localJointOffset=<5 -1 -0.2>
     }]
    relativeSpaceBone {boneName=spine_05}
   }
   n7 = Anim Dynamics - Bone: dyn_HairB01_r(ComponentPose=@n6, Alpha=0.5, ExternalForce="0,0,0") {
    angularDampingOverride=0.999
    angularSpringConstant=800
    bAngularSpring=true
    bGravityOverrideInSimSpace=true
    bOverrideAngularDamping=true
    bUseGravityOverride=true
    gravityScale=0.2
    simulationSpace=RootRelative
    alphaBoolBlend {blendOption=Linear}
    boundBone {boneName=dyn_HairB01_r}
    componentAppliedLinearAccClamp=<3000 3000 3000>
    componentLinearAccScale=<0.8 0.8 0.8>
    componentLinearVelScale=<0.8 0.8 0.8>
    gravityOverride {z=-980}
    physicsBodyDefinitions [{
      sphereCollisionRadius=0
      boundBone {boneName=dyn_HairB01_r}
      boxExtents=<30 30 30>
      constraintSetup {angularTargetAxis=AxisY, angularLimitsMax=<5 10 45>, angularLimitsMin=<-5 -10 -7>, angularTarget=<0 0 0>}
      localJointOffset=<5 -0.5 0.5>
     }]
    relativeSpaceBone {boneName=spine_05}
   }
   n8 = Anim Dynamics - Bone: dyn_HairB02_r(ComponentPose=@n7, Alpha=0.5, ExternalForce="0,0,0") {
    angularDampingOverride=0.999
    angularSpringConstant=800
    bAngularSpring=true
    bGravityOverrideInSimSpace=true
    bOverrideAngularDamping=true
    bUseGravityOverride=true
    gravityScale=0.2
    simulationSpace=RootRelative
    alphaBoolBlend {blendOption=Linear}
    boundBone {boneName=dyn_HairB02_r}
    componentAppliedLinearAccClamp=<3000 3000 3000>
    componentLinearAccScale=<0.8 0.8 0.8>
    componentLinearVelScale=<0.8 0.8 0.8>
    gravityOverride {z=-980}
    physicsBodyDefinitions [{
      sphereCollisionRadius=0
      boundBone {boneName=dyn_HairB02_r}
      boxExtents=<30 30 30>
      constraintSetup {angularTargetAxis=AxisY, angularLimitsMax=<5 10 45>, angularLimitsMin=<-5 -10 -7>, angularTarget=<0 0 0>}
      localJointOffset=<5 -0.5 0.5>
     }]
    relativeSpaceBone {boneName=spine_05}
   }
   Output Pose(Result=Component To Local(ComponentPose=@n8)) {layerGroup=None}
  ]}
 EventGraph {nodes [Event BlueprintUpdateAnimation()]}
]
variables []
